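# Niagara VFX for the PearlTaffyVista emote: glass shattering, followed by a delayed burst of fire.
# Glass maps spawn index to 22 mesh pieces, with velocity, gravity and spin chosen per piece, plus a brief flash.
# BreakingFlames and FireWall start after System.FireDelay, adding flame clusters, a wall of sprites and timed lights.
# Fire, also delayed, trails particles of the Rose emitter, with front-end preview and time-of-day color adjustments.
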NiagaraSystem /BRCosmetics/Animation/Game/MainPlayer/Emotes/PearlTaffyVista/VFX/NS_Emote_PearlTaffyVista
dependencies {
 usedDataInterfaces [
  &NiagaraDataInterfaceCurve
  &NiagaraDataInterfaceVectorCurve
  &NiagaraDataInterfaceColorCurve
  &NiagaraDataInterfaceSpriteRendererInfo
  &NiagaraDataInterfaceParticleRead
  &NiagaraDataInterfaceCamera
  &NiagaraDataInterfaceArrayInt32
 ]
 usedDynamicInputs [&MaskVector2dBySpawnGroup]
 usedModules [
  &self:SystemSpawnScript.SystemScriptSource.SystemScriptGraph.NiagaraNodeAssignment_2.SetVariables_#1
  &SystemState
  &self:SystemSpawnScript.SystemScriptSource.SystemScriptGraph.NiagaraNodeAssignment_0.SetVariables_#2
  &self:SystemSpawnScript.SystemScriptSource.SystemScriptGraph.NiagaraNodeAssignment_1.SetVariables_#3
 ]
}
emitters {
 BreakingFlames {
  emitterUpdateScript [
   &EmitterState(Life Cycle Mode=::NewEnumerator1, Loop Behavior=::NewEnumerator1, Loop Delay=$System.FireDelay, Loop Duration=2, UseLoopDelay=true)
   &SpawnBurst_Instantaneous(Spawn Count=25, Use Spawn Probability=false, UseLoopCountLimit=false)
   &SpawnRate(
    SpawnRate=&FloatFromCurve(FloatCurve=DataInterface() {propertyValues {Curve {keys [<at 0.27 1>,<at 0.36 0>]}}}, CurveIndex=$Emitter.NormalizedLoopAge, Scale Curve=25)
    Use Spawn Probability=false
   )
  ]
  particleSpawnScript [
   &PartitionParticles(Number Of Partitions=5) {module=PartitionParticles001}
   &PartitionParticles(Number Of Partitions=&MakeCustomIntFromBool(
     Bool=false
     True INT=&MakeCustomIntFromBool(Bool=false, True INT=1, False INT=2)
     False INT=&MakeCustomIntFromBool(Bool=false, True INT=2, False INT=4)
    ))
   &InitializeParticle(
    Color Maximum=<rgba 6.0313 0.2578 0 1.2>
    Color Minimum=<rgba 10 0.7456 0 1.2>
    Color Mode=::NewEnumerator2
    Lifetime Max=0.65
    Lifetime Min=0.5
    Lifetime Mode=::NewEnumerator1
    Sprite Rotation Mode=::NewEnumerator2
    Sprite Size Mode=::NewEnumerator1
    Uniform Sprite Size Max=40
    Uniform Sprite Size Min=3
    UsePositionOffset=false
    Write Material Random=true
   )
   &ShapeLocation(
    Box Size=<20 10 40>
    MaskSpawnGroup=false
    Offset=&MergeHelper_VectorSwitch(
     0=<-70 -40 40>
     1=<-25 -40 60>
     2=<70 -40 50>
     3=<55 -40 45>
     4=<-55 -40 50>
     Index=$Particles.PartitionParticles.Partition
     Num=::NewEnumerator3
    )
    Offset Mode=::NewEnumerator0
    Override Seed=false
    Shape Primitive=::NewEnumerator2
    Surface Only=false
    UseOwnerScale=true
   )
   &AddVelocity(
    Cone Axis=<0 0 1>
    Module.UseConeFalloff=false
    Module.UseVelDistribution=false
    Override Seed=false
    Velocity Mode=::NewEnumerator2
    Velocity Speed=&RandomRangeFloat(Minimum=100, Maximum=300)
   )
   &ApplyInitialForces()
  ]
  particleUpdateScript [
   &ParticleState()
   &ScaleSpriteSize(Uniform Curve Sprite Scale=DataInterface() {propertyValues {Curve {keys [{tangentMode=RCTM_User, time=0, value=0.6}, {tangentMode=RCTM_User, time=1, value=1.5}]}}})
   &SpriteRotationRate(Rotation Rate=&Multiply_Float(
     A=&RandomRangeFloat(Minimum=20, Maximum=30)
     B=&MakeCustomFloatFromBool(Bool=&RandomBool(), True Float=1, False Float=-1)
    ))
   &SubUVAnimation(
    "Sprite Renderer "=DataInterface() {propertyValues {SpriteRenderer=&self:BreakingFlames_0.NiagaraSpriteRendererProperties_0}}
    UseEndFrame=false
    UseStartFrame=false
   )
   &ScaleColor(
    Scale Alpha=&Multiply_Float(
     A=1
     B=&FloatFromCurve(
      FloatCurve=DataInterface() {propertyValues {Curve {keys [
          <at 0 0 RCIM_Cubic>
          {arriveTangent=-0.0116, interpMode=RCIM_Cubic, leaveTangent=-0.0116, tangentMode=RCTM_User, time=0.5, value=1}
          {arriveTangent=-2.686, interpMode=RCIM_Cubic, leaveTangent=-2.686, tangentMode=RCTM_User, time=0.6746, value=0.6742}
          {interpMode=RCIM_Cubic, leaveTangent=-2.371, time=1, value=0}
         ]}}}
      Scale Curve=2
     )
    )
    Scale RGB=&VectorFromCurve(VectorCurve=DataInterface() {propertyValues {
       XCurve {keys [<at 0.3242 3 RCIM_Cubic>,<at 0.9878 0.9731>]}
       YCurve {keys [<at 0.3242 2.8 RCIM_Cubic>,<at 1 0.9731>]}
       ZCurve {keys [<at 0.3169 2.6605 RCIM_Cubic>,<at 1 0.9731>]}
      }})
    ScaleA=true
    ScaleRGB=true
   )
   &DynamicMaterialParameters(
    Index 0 Param 1=1
    Index 0 Param 2=&RandomRangeFloat(Minimum=0, Maximum=4)
    Index 0 Param 3=&FloatFromCurve(FloatCurve=DataInterface() {propertyValues {Curve {keys [<at 0 0.5 RCIM_Cubic>, {arriveTangent=1.5, interpMode=RCIM_Cubic, leaveTangent=1.5, time=0.5265, value=0.7148}, <at 1 2>]}}})
    Index 0 Param 4=2.5
    Module.Index0Param0Write=true
    Module.Index0Param1Write=true
    Module.Index0Param2Write=true
    Module.Index0Param3Write=true
   )
   &CurlNoiseForce(Bias Noise=false, Noise Frequency=5, Noise Strength=150, Pan Noise=false)
   &SolveForcesAndVelocity(Clamp Velocity=false, Limit Acceleration=false)
  ]
  rendererClasses [&NiagaraSpriteRendererProperties]
  renderers [{rendererClass=&NiagaraSpriteRendererProperties}]
 }
 Fire {
  emitterUpdateScript [
   &EmitterState(Life Cycle Mode=::NewEnumerator1, Loop Behavior=::NewEnumerator1, Loop Delay=$System.FireDelay, Loop Duration=2, UseLoopDelay=true)
   &SpawnParticlesFromOtherEmitter(
    Attribute Reader=DataInterface() {propertyValues {EmitterBinding {bindingMode=Other, emitterName=Rose}}}
    Spawn Rate=15
    Use Maximum Age=false
    Use Minimum Age=false
   )
  ]
  particleSpawnScript [
   &InitializeParticle(
    Lifetime Max=1.2
    Lifetime Min=0.4
    Lifetime Mode=::NewEnumerator1
    Sprite Rotation Mode=::NewEnumerator2
    Sprite Size Mode=::NewEnumerator1
    Uniform Sprite Size Max=30
    UsePositionOffset=false
    Write Material Random=true
   )
   &SampleParticlesFromOtherEmitter(Velocity Sampling=::NewEnumerator0)
  ]
  particleUpdateScript [
   &ParticleState()
   &SpriteRotationRate(Rotation Rate=&RandomRangeFloat(Minimum=-100, Maximum=100))
   &ScaleSpriteSize(
    Uniform Curve Scale=1.3
    Uniform Curve Sprite Scale=DataInterface() {propertyValues {Curve {keys [
        {arriveTangent=2.2767, interpMode=RCIM_Cubic, leaveTangent=2.2767, tangentMode=RCTM_User, time=0, value=0.2}
        {arriveTangent=-0.009, interpMode=RCIM_Cubic, leaveTangent=-0.009, tangentMode=RCTM_User, time=1, value=1}
       ]}}}
   )
   &Drag(Use Linear Drag=true, Use Rotational Drag=true)
   &SolveForcesAndVelocity(Clamp Velocity=false, Limit Acceleration=false)
   &ScaleColor(
    Scale Alpha=&FloatFromCurve(
     FloatCurve=DataInterface() {propertyValues {Curve {keys [
         <at 0 0 RCIM_Cubic>
         {arriveTangent=-0.3533, interpMode=RCIM_Cubic, leaveTangent=-0.3533, tangentMode=RCTM_User, time=0.22, value=1}
         {arriveTangent=-1.4952, interpMode=RCIM_Cubic, leaveTangent=-1.4952, tangentMode=RCTM_User, time=0.8, value=0.5}
         <at 1 0 RCIM_Cubic>
        ]}}}
     Scale Curve=&MakeCustomFloatFromBool(Bool=$User.bIsFrontEndPreview, True Float=0.7, False Float=&RandomRangeFloat(Minimum=0.5, Maximum=1))
    )
    Scale RGB=&MakeCustomVectorFromBool(
     Bool=$User.bIsFrontEndPreview
     True Vector=&MakeVectorFromLinearColorRGB(Color=&ColorFromCurve(ColorCurve=DataInterface() {propertyValues {
         AlphaCurve {keys []}
         BlueCurve {keys [<at 0.0717 0.0133>, {arriveTangent=-0.0287, interpMode=RCIM_Cubic, leaveTangent=-0.0287, time=0.4137, value=0.01}, <at 0.536 0>]}
         GreenCurve {keys [<at 0.0717 0.22>, {arriveTangent=-0.4631, interpMode=RCIM_Cubic, leaveTangent=-0.4631, time=0.4137, value=0.075}, <at 0.536 0.005>]}
         RedCurve {keys [<at 0.0717 1>, {arriveTangent=-2.0644, interpMode=RCIM_Cubic, leaveTangent=-2.0644, time=0.4137, value=0.5}, <at 0.536 0.0417>]}
        }}))
     False Vector=&MakeVectorFromLinearColorRGB(Color=&ColorFromCurve(ColorCurve=DataInterface() {propertyValues {
         AlphaCurve {keys []}
         BlueCurve {keys [<at 0.0392 0.0243>,<at 0.1766 0 RCIM_Cubic>,<at 0.5058 0>]}
         GreenCurve {keys [<at 0.0392 0.3>,<at 0.1766 0.12 RCIM_Cubic>,<at 0.5058 0.005>]}
         RedCurve {keys [<at 0.0392 1.5>,<at 0.1766 1 RCIM_Cubic>,<at 0.5058 0.0417>]}
        }}))
    )
    ScaleA=true
    ScaleRGB=true
   )
   &ScaleColor(
    Scale Alpha=&FNTimeOfDayScalar(Night Scale=0.6, Morning Scale=0.7, Evening Scale=0.7)
    Scale RGB=&MakeVectorFromLinearColorRGB(Color=&FNTimeOfDayColor(Night Color=<rgba 1 0.9948 0.2694 1>, Morning Color=<rgba 1 1 1 1>, Day Color=<rgba 1 1 1 1>, Evening Color=<rgba 1 1 1 1>))
    ScaleA=true
    ScaleRGB=true
   ) {module=ScaleColor001}
   &SubUVAnimation(
    End Frame Range Override:NiagaraInt32=&RandomRangeInt(Minimum=12, Maximum=16)
    "Sprite Renderer "=DataInterface() {propertyValues {SpriteRenderer=&self:Fire_1.NiagaraSpriteRendererProperties_2}}
    Start Frame Range Override=7
    UseEndFrame=true
    UseStartFrame=true
   )
  ]
  rendererClasses [&NiagaraSpriteRendererProperties]
  renderers [{rendererClass=&NiagaraSpriteRendererProperties}]
 }
 FireWall {
  emitterUpdateScript [
   &EmitterState(Life Cycle Mode=::NewEnumerator1, Loop Behavior=::NewEnumerator1, Loop Delay=$System.FireDelay, Loop Duration=2, UseLoopDelay=true)
   &SpawnBurst_Instantaneous(Spawn Count=30, Use Spawn Probability=false, UseLoopCountLimit=false)
   &SpawnRate(SpawnRate=200, Use Spawn Probability=false)
   &SpawnBurst_Instantaneous(Spawn Group=1, Use Spawn Probability=false, UseLoopCountLimit=false) {module=SpawnBurst_Instantaneous001}
   &SpawnBurst_Instantaneous(Spawn Count=30, Spawn Group=2, Use Spawn Probability=false, UseLoopCountLimit=false) {module=SpawnBurst_Instantaneous002}
   &SpawnRate(SpawnGroup=2, SpawnRate=50, Use Spawn Probability=false) {module=SpawnRate001}
  ]
  particleSpawnScript [
   /SetVariables(Particles.VisibilityTag=&SpawnGroup()) {isSetParameters=true}
   &PartitionParticles(Number Of Partitions=5)
   &InitializeParticle(
    Lifetime=&MergeHelper_FloatSwitch(
     0=&UniformAOrBFloat(
      A=&RandomRangeFloat(Minimum=0.5, Maximum=0.7)
      B=&RandomRangeFloat(Minimum=0.8, Maximum=1)
      Distribution Weight=0.2
      Evaluation Type=::NewEnumerator0
      Recalculate Random Each Loop=true
      Override Seed=false
     )
     1=&RandomRangeFloat(Minimum=0.5, Maximum=0.6)
     2=?
     3=?
     Index=&SpawnGroup()
     Num=::NewEnumerator0
    )
    Mass Mode=::NewEnumerator3
    Sprite Rotation Mode=::NewEnumerator2
    Sprite Size Mode=::NewEnumerator4
    UsePositionOffset=false
    Write Material Random=true
   )
   &ShapeLocation(
    Box Size=&MaskVector(Mask Source=0, Mask Parameter=?, Mask=2, Masked Vector=<70 10 100>, Passthrough Vector=<160 10 130>)
    MaskSpawnGroup=false
    Offset=<0 -40 80>
    Offset Mode=::NewEnumerator0
    Override Seed=false
    Shape Origin=$Engine.Emitter.SimulationPosition
    Shape Primitive=::NewEnumerator2
    Surface Only=false
    UseOwnerScale=true
   ) {module=ShapeLocation001}
   /SetVariables(Particles.Init Fire Sprite Size:Vector2f=&MakeCustomVector2DFromBool(
     Bool=&SetBoolByFloatComparison(A=&RandomRangeFloat(Minimum=0, Maximum=1.5), B=0.5)
     True Vector2D=&RandomRangeVector2D(Minimum=<35 35>, Maximum=<50 50>)
     False Vector2D=&MakeCustomVector2DFromBool(
      Bool=&SetBoolByFloatComparison(A=&RandomRangeFloat(Minimum=0, Maximum=1.5), B=1)
      True Vector2D=&Vector2DFromFloat(Value=&RandomRangeFloat(Minimum=13, Maximum=65))
      False Vector2D=<25 50>
     )
    )) {isSetParameters=true}
   &AddVelocity(Cone Axis=<0 0 1>, Module.UseConeFalloff=false, Module.UseVelDistribution=false, Override Seed=false, Velocity Mode=::NewEnumerator2)
  ]
  particleUpdateScript [
   &ParticleState()
   &GravityForce(Gravity=&Lerp_Vector(
     A=&RandomRangeVector(Minimum=<0 0 200>, Maximum=<0 0 150>)
     B=<0 0 500>
     Alpha=&FloatFromCurve(
      FloatCurve=DataInterface() {propertyValues {Curve {
         defaultValue=1
         keys [{interpMode=RCIM_Cubic, tangentMode=RCTM_User, time=0.25, value=0}, {interpMode=RCIM_Cubic, tangentMode=RCTM_User, time=0.9, value=1}]
        }}}
      CurveIndex=$Emitter.Age
     )
    ))
   &CurlNoiseForce(Bias Noise=false, Noise Frequency=25, Noise Strength=125, Pan Noise=false)
   &VortexForce(
    UseOriginOffset=false
    Vortex Force Amount=&FloatFromCurve(
     FloatCurve=DataInterface() {propertyValues {Curve {defaultValue=1, keys [<at 0.0496 0 RCIM_Cubic>,<at 0.7818 1 RCIM_Cubic>]}}}
     Scale Curve=200
    )
   )
   &SolveForcesAndVelocity(Clamp Velocity=false, Limit Acceleration=false)
   /SetVariables(
    Particles.Fire Dynamic Param:Vector4f=&MakeVector4(
     X=&FloatFromCurve(
      FloatCurve=DataInterface() {propertyValues {Curve {
         defaultValue=1
         keys [{interpMode=RCIM_Cubic, tangentMode=RCTM_User, time=0, value=0}, {interpMode=RCIM_Cubic, tangentMode=RCTM_User, time=1.0015, value=0.6706}]
        }}}
      Scale Curve=&RandomRangeFloat(Minimum=0.85, Maximum=1.2)
     )
     Y=&RandomRangeFloat(Minimum=0, Maximum=4)
     Z=0
     W=2.5
    )
    Particles.Fire Sprites Color:LinearColor=&MakeLinearColorFromVectorAndFloat(
     "Vector (RGB)"=&Multiply_VectorByFloat(
      Vector=&RandomRangeVector(
       Minimum=&VectorFromCurve(
        VectorCurve=DataInterface() {propertyValues {
          XCurve {keys [
            <at 0 0 RCIM_Cubic>
            {arriveTangent=0.9616, tangentMode=RCTM_User, time=0.3523, value=0.7057}
            <at 0.4794 0.8221>
            {arriveTangent=-1.3705, interpMode=RCIM_Cubic, leaveTangent=-1.3705, tangentMode=RCTM_User, time=0.6961, value=0.308}
            {arriveTangent=-0.4056, interpMode=RCIM_Cubic, leaveTangent=-0.4056, tangentMode=RCTM_User, time=1, value=0}
           ]}
          YCurve {keys [
            <at 0 0 RCIM_Cubic>
            <at 0.4077 1>
            {arriveTangent=-1.4704, interpMode=RCIM_Cubic, leaveTangent=-1.4704, tangentMode=RCTM_User, time=0.7003, value=0.2971}
            {arriveTangent=-0.4056, interpMode=RCIM_Cubic, leaveTangent=-0.4056, tangentMode=RCTM_User, time=1, value=0}
           ]}
          ZCurve {keys [
            <at 0 0 RCIM_Cubic>
            <at 0.4032 1>
            {arriveTangent=-1.4587, interpMode=RCIM_Cubic, leaveTangent=-1.4587, tangentMode=RCTM_User, time=0.6972, value=0.296}
            {arriveTangent=-0.4056, interpMode=RCIM_Cubic, leaveTangent=-0.4056, tangentMode=RCTM_User, time=1, value=0}
           ]}
         }}
        Scale Curve=&Lerp_Vector(A=<1.15 1 1>, B=<0.6 0.6 0.6>, Alpha=&RandomRangeFloat(Minimum=0, Maximum=1))
       )
       Maximum=&VectorFromCurve(
        VectorCurve=DataInterface() {propertyValues {
          XCurve {keys [
            <at 0 0 RCIM_Cubic>
            {arriveTangent=0.9616, tangentMode=RCTM_User, time=0.3523, value=0.7057}
            <at 0.4794 0.8221>
            {arriveTangent=-1.5791, interpMode=RCIM_Cubic, leaveTangent=-1.5791, time=0.6756, value=0.1581}
            <at 1 0 RCIM_Cubic>
           ]}
          YCurve {keys [
            <at 0 0 RCIM_Cubic>
            <at 0.4077 1>
            {arriveTangent=-1.6883, interpMode=RCIM_Cubic, leaveTangent=-1.6883, time=0.6798, value=0.1472}
            <at 1 0 RCIM_Cubic>
           ]}
          ZCurve {keys [
            <at 0 0 RCIM_Cubic>
            <at 0.4032 1>
            {arriveTangent=-1.6755, interpMode=RCIM_Cubic, leaveTangent=-1.6755, time=0.6767, value=0.1461}
            <at 1 0 RCIM_Cubic>
           ]}
         }}
        Scale Curve=&Lerp_Vector(A=<1.15 1 1>, B=<0.6 0.6 0.6>, Alpha=&RandomRangeFloat(Minimum=0, Maximum=1))
       )
      )
      Float=&FloatFromCurve(
       FloatCurve=DataInterface() {propertyValues {Curve {keys [<at -0.05 0>,<at -0.05 1>,<at -0.05 2>,<at 0.0186 0.25>,<at 0.0295 5>,<at 0.0428 5>,<at 0.0459 2>]}}}
       CurveIndex=$Emitter.Age
       Scale Curve=&RandomRangeFloat(Minimum=0.4, Maximum=1.2)
      )
     )
     "Float (Alpha)"=&Multiply_Float(
      A=&FloatFromCurve(
       FloatCurve=DataInterface() {propertyValues {Curve {keys [
           {arriveTangent=3.3783, interpMode=RCIM_Cubic, leaveTangent=3.3783, tangentMode=RCTM_User, time=0, value=0}
           {arriveTangent=0.2941, interpMode=RCIM_Cubic, leaveTangent=0.2941, tangentMode=RCTM_User, time=0.203, value=0.7175}
           {arriveTangent=-1.109, interpMode=RCIM_Cubic, leaveTangent=-1.109, time=0.5122, value=0.5742}
           {arriveTangent=-0.0646, interpMode=RCIM_Cubic, leaveTangent=-0.0646, tangentMode=RCTM_User, time=0.85, value=0}
          ]}}}
       Scale Curve=0.3
      )
      B=&FloatFromCurve(FloatCurve=DataInterface() {propertyValues {Curve {keys [<at 1.1 1>,<at 1.46 0>]}}}, CurveIndex=$Emitter.Age)
     )
    )
    Particles.Update Fire Sprite Size:Vector2f=&Multiply_Vector2D(
     A=$Particles.Init Fire Sprite Size
     B=&Multiply_Vector2D_ByFloat(
      Float=&FloatFromCurve(
       FloatCurve=DataInterface() {propertyValues {Curve {keys [
           {arriveTangent=1.1411, interpMode=RCIM_Cubic, leaveTangent=1.1411, tangentMode=RCTM_User, time=0, value=1.5}
           {arriveTangent=0.7068, interpMode=RCIM_Cubic, leaveTangent=0.7068, tangentMode=RCTM_User, time=0.2325, value=1.747}
           {arriveTangent=-0.4639, interpMode=RCIM_Cubic, leaveTangent=-0.4639, time=0.6, value=1.8}
           {interpMode=RCIM_Cubic, tangentMode=RCTM_User, time=1.0543, value=1.3657}
          ]}}}
       Scale Curve=&RandomRangeFloat(Minimum=0.15, Maximum=1.1)
      )
      Vector2D=&RandomRangeVector2D(Minimum=<0.7 0.7>, Maximum=<1 1.7>)
     )
    )
   ) {isSetParameters=true}
   /SetVariables(
    Particles.Dark Fire Color:LinearColor=&MakeLinearColorFromVectorAndFloat(
     "Vector (RGB)"=&Multiply_VectorByFloat(
      Vector=<0.12 0.25 0.25>
      Float=&FloatFromCurve(FloatCurve=DataInterface() {propertyValues {Curve {keys [<at 0 1>,<at 1 0>]}}})
     )
     "Float (Alpha)"=&Multiply_Float(
      A=&FloatFromCurve(
       FloatCurve=DataInterface() {propertyValues {Curve {
          defaultValue=1
          keys [
           {interpMode=RCIM_Cubic, tangentMode=RCTM_User, time=0, value=0}
           {interpMode=RCIM_Cubic, tangentMode=RCTM_User, time=0.5969, value=1}
           {tangentMode=RCTM_User, time=0.9832, value=0}
          ]
         }}}
       Scale Curve=0.2
      )
      B=&FloatFromCurve(FloatCurve=DataInterface() {propertyValues {Curve {keys [<at 1.1 1>,<at 1.46 0>]}}}, CurveIndex=$Emitter.Age)
     )
    )
    Particles.Dark Fire Size:Vector2f=&Multiply_Vector2D_ByFloat(
     Float=&FloatFromCurve(
      FloatCurve=DataInterface() {propertyValues {Curve {
         defaultValue=1
         keys [{interpMode=RCIM_Cubic, tangentMode=RCTM_User, time=0.5476, value=1}, {interpMode=RCIM_Cubic, tangentMode=RCTM_User, time=1, value=1.5}]
        }}}
      Scale Curve=&RandomRangeFloat(Minimum=0.5, Maximum=1)
     )
     Vector2D=$Particles.Update Fire Sprite Size
    )
   ) {isSetParameters=true}
   /SetVariables(Particles.Center Color:LinearColor=&MakeLinearColorFromVectorAndFloat(
     "Vector (RGB)"=&Multiply_VectorByFloat(
      Vector=&VectorFromCurve(
       VectorCurve=DataInterface() {propertyValues {
         XCurve {keys [
           <at 0 0 RCIM_Cubic>
           {arriveTangent=0.4425, interpMode=RCIM_Cubic, leaveTangent=0.4425, time=0.399, value=1.0109}
           {arriveTangent=-1.3705, interpMode=RCIM_Cubic, leaveTangent=-1.3705, tangentMode=RCTM_User, time=0.6961, value=0.308}
           {arriveTangent=-0.4056, interpMode=RCIM_Cubic, leaveTangent=-0.4056, tangentMode=RCTM_User, time=1, value=0}
          ]}
         YCurve {keys [
           <at 0 0 RCIM_Cubic>
           <at 0.4077 1>
           {arriveTangent=-1.4704, interpMode=RCIM_Cubic, leaveTangent=-1.4704, tangentMode=RCTM_User, time=0.7003, value=0.2971}
           {arriveTangent=-0.4056, interpMode=RCIM_Cubic, leaveTangent=-0.4056, tangentMode=RCTM_User, time=1, value=0}
          ]}
         ZCurve {keys [
           <at 0 0 RCIM_Cubic>
           {arriveTangent=5.1821, tangentMode=RCTM_User, time=0.4032, value=1}
           {arriveTangent=-1.4587, interpMode=RCIM_Cubic, leaveTangent=-1.4587, tangentMode=RCTM_User, time=0.6972, value=0.296}
           {arriveTangent=-0.4056, interpMode=RCIM_Cubic, leaveTangent=-0.4056, tangentMode=RCTM_User, time=1, value=0}
          ]}
        }}
       Scale Curve=&RandomRangeVector(Minimum=<1 1.1 1>, Maximum=<1.2 1.6 1>)
      )
      Float=&FloatFromCurve(
       FloatCurve=DataInterface() {propertyValues {Curve {keys [<at 0.0186 0.25>,<at 0.0295 3>,<at 0.0428 3>,<at 0.0459 2>,<at 1.1552 1>,<at 1.387 0.5>,<at 1.65 0>]}}}
       CurveIndex=$Emitter.Age
       Scale Curve=&RandomRangeFloat(Minimum=0.85, Maximum=1.5)
      )
     )
     "Float (Alpha)"=&Multiply_Float(
      A=&FloatFromCurve(
       FloatCurve=DataInterface() {propertyValues {Curve {keys [
           {arriveTangent=3.3783, interpMode=RCIM_Cubic, leaveTangent=3.3783, tangentMode=RCTM_User, time=0, value=0}
           {arriveTangent=0.2941, interpMode=RCIM_Cubic, leaveTangent=0.2941, tangentMode=RCTM_User, time=0.203, value=0.7175}
           {arriveTangent=-1.109, interpMode=RCIM_Cubic, leaveTangent=-1.109, time=0.5122, value=0.5742}
           {arriveTangent=-0.0646, interpMode=RCIM_Cubic, leaveTangent=-0.0646, tangentMode=RCTM_User, time=0.85, value=0}
          ]}}}
       Scale Curve=&RandomRangeFloat(Minimum=0.25, Maximum=0.8)
      )
      B=&FloatFromCurve(FloatCurve=DataInterface() {propertyValues {Curve {keys [<at 1.1 1>,<at 1.46 0>]}}}, CurveIndex=$Emitter.Age)
     )
    )) {isSetParameters=true}
   &Light_Attributes(
    Diffuse Scale=1
    Falloff Exponent=1
    Light Position Offset:Vector3f=<0 -40 100>
    Position=1
    Position to Offset:NiagaraPosition=$Engine.Emitter.SimulationPosition
    Radius=1
    Specular Scale=1
    UsePositionOffset=true
    Volumetric Scattering=1
   )
   /SetVariables(Particles.Flame Light Color:LinearColor=&Multiply_LinearColorByFloat(
     Linear Color=&ColorFromCurve(ColorCurve=DataInterface() {propertyValues {
        AlphaCurve {keys [<at 0 0 RCIM_Cubic>,<at 0.1036 1>,<at 0.7529 0.25>,<at 1 0 RCIM_Cubic>]}
        BlueCurve {keys [
          <at 0 0>
          <at 0.0478 0>
          <at 0.0914 0>
          {arriveTangent=4.4329, interpMode=RCIM_Cubic, leaveTangent=4.4329, time=0.1363, value=0.6142}
          <at 0.2357 0.8327>
          {arriveTangent=-3.5063, interpMode=RCIM_Cubic, leaveTangent=-3.5063, time=0.3241, value=0.0725}
          <at 0.4562 0.0595>
          <at 0.5488 0>
          <at 0.9304 0>
         ]}
        GreenCurve {keys [
          <at 0 0.0554>
          <at 0.0478 0.2567>
          <at 0.0914 0.7515>
          {arriveTangent=9.2648, interpMode=RCIM_Cubic, leaveTangent=9.2648, time=0.1363, value=2.4314}
          <at 0.2357 1.9971>
          {arriveTangent=-7.0261, interpMode=RCIM_Cubic, leaveTangent=-7.0261, time=0.3241, value=0.6629}
          <at 0.4562 0.4477>
          <at 0.5488 0.2631>
          <at 0.9304 0.0786>
         ]}
        RedCurve {keys [
          <at 0 0.2156>
          <at 0.0478 1>
          <at 0.0914 2.011>
          {arriveTangent=10.6147, interpMode=RCIM_Cubic, leaveTangent=10.6147, time=0.1363, value=2.994}
          <at 0.2357 2.994>
          {arriveTangent=-9.0422, interpMode=RCIM_Cubic, leaveTangent=-9.0422, time=0.3241, value=1}
          <at 0.4562 1>
          <at 0.5488 1>
          <at 0.9304 0.449>
         ]}
       }})
     Float=8
    )) {isSetParameters=true}
   &Light_Attributes(
    Diffuse Scale=1
    Falloff Exponent=1
    Light Position Offset:Vector3f=<0 60 130>
    Position=1
    Position to Offset:NiagaraPosition=$Engine.Emitter.SimulationPosition
    Radius=1
    Specular Scale=1
    UsePositionOffset=true
    Volumetric Scattering=1
   ) {module=Light_Attributes001}
   /SetVariables(Particles.Front Light Color:LinearColor=&Multiply_LinearColorByFloat(
     Linear Color=&ColorFromCurve(ColorCurve=DataInterface() {propertyValues {
        AlphaCurve {keys [
          <at 0.0002 0>
          {arriveTangent=0.3975, interpMode=RCIM_Cubic, leaveTangent=0.3975, time=0.0873, value=1}
          <at 0.6291 0.25>
          <at 1 0 RCIM_Cubic>
         ]}
        BlueCurve {keys [<at 0.0002 0>,<at 0.0805 0 RCIM_Cubic>,<at 0.2765 0>,<at 0.4929 0 RCIM_Cubic>]}
        GreenCurve {keys [
          <at 0.0002 0.1153>
          {arriveTangent=0.4819, interpMode=RCIM_Cubic, leaveTangent=0.4819, time=0.0805, value=0.5225}
          <at 0.2765 0.2485>
          <at 0.4929 0.096 RCIM_Cubic>
         ]}
        RedCurve {keys [
          <at 0.0002 1>
          {arriveTangent=-0.3958, interpMode=RCIM_Cubic, leaveTangent=-0.3958, time=0.0805, value=1}
          <at 0.2765 0.8906>
          <at 0.4929 0.8323 RCIM_Cubic>
         ]}
       }})
     Float=2.5
    )) {isSetParameters=true}
  ]
  rendererClasses [&NiagaraSpriteRendererProperties, &NiagaraLightRendererProperties]
  renderers [
   {rendererClass=&NiagaraSpriteRendererProperties}
   {rendererClass=&NiagaraSpriteRendererProperties, rendererIndex=1}
   {rendererClass=&NiagaraLightRendererProperties, rendererIndex=2}
   {rendererClass=&NiagaraLightRendererProperties, rendererIndex=3}
   {rendererClass=&NiagaraSpriteRendererProperties, rendererIndex=4}
  ]
 }
 Glass {
  emitterUpdateScript [
   &EmitterState(Life Cycle Mode=::NewEnumerator1, Loop Behavior=::NewEnumerator1, Loop Duration=2, UseLoopDelay=false)
   &SpawnBurst_Instantaneous(Spawn Count=21, Use Spawn Probability=false, UseLoopCountLimit=false)
   &SpawnBurst_Instantaneous(Spawn Group=1, Use Spawn Probability=false, UseLoopCountLimit=false) {module=SpawnBurst_Instantaneous001}
   /SetVariables(Emitter.Opacity:NiagaraFloat=&FloatFromCurve(
     FloatCurve=DataInterface() {propertyValues {Curve {keys [
         {arriveTangent=-4.7558, interpMode=RCIM_Cubic, leaveTangent=-4.7558, tangentMode=RCTM_User, time=1, value=0.8}
         {arriveTangent=-1.0407, interpMode=RCIM_Cubic, leaveTangent=-1.0407, tangentMode=RCTM_User, time=1.2, value=0.3462}
         {arriveTangent=-0.7988, interpMode=RCIM_Cubic, leaveTangent=-0.7988, tangentMode=RCTM_User, time=1.3674, value=0}
        ]}}}
     CurveIndex=$Emitter.Age
    )) {isSetParameters=true}
   /SetVariables(Emitter.Metallicness:NiagaraFloat=&MakeCustomFloatFromBool(
     Bool=&Boolean_OrOperation(Bool1=$User.bIsFrontEnd, Bool2=$User.bIsFrontEndPreview)
     True Float=0.85
     False Float=1
    )) {isSetParameters=true}
  ]
  particleSpawnScript [
   /SetVariables(Particles.VisibilityTag=&SpawnGroup()) {isSetParameters=true}
   /SetVariables(Particles.MeshIndex=&SelectIntFromArray(
     Direct Array Index=&ReturnExecIndex()
     Int Selection Array=DataInterface() {propertyValues {IntData [0,1,2,3,4,5,6,7,8,9,10,11,12,13,14,15,16,17,18,19,20,21]}}
    )) {isSetParameters=true}
   &InitializeParticle(
    Color=&MergeHelper_ColorSwitch(
     0=<rgba 1 1 1 1>
     1=<rgba 1 0.5055 0 1>
     2=<rgba 0 0 0 0>
     3=<rgba 0 0 0 0>
     4=<rgba 0 0 0 0>
     Index=&SpawnGroup()
     Num=::NewEnumerator3
    )
    Lifetime=&MergeHelper_FloatSwitch(
     0=&MakeCustomFloatFromBool(
      Bool=&SetBoolByIntComparison(A=$Particles.MeshIndex, B=8, Comparison Type=::NewEnumerator6)
      True Float=&RandomRangeFloat(Minimum=0.15, Maximum=0.3)
      False Float=&MakeCustomFloatFromBool(
       Bool=&Boolean_AndOperation(
        Bool1=&SetBoolByIntComparison(A=$Particles.MeshIndex, B=8, Comparison Type=::NewEnumerator2)
        Bool2=&SetBoolByIntComparison(A=$Particles.MeshIndex, B=12, Comparison Type=::NewEnumerator6)
       )
       True Float=&RandomRangeFloat(Minimum=0.25, Maximum=0.4)
       False Float=&RandomRangeFloat(Minimum=1, Maximum=1.6)
      )
     )
     1=0.1
     2=0
     3=0
     4=0
     Index=&SpawnGroup()
     Num=::NewEnumerator3
    )
    Mass Mode=::NewEnumerator3
    Position Offset=&MergeHelper_VectorSwitch(0=<0 90 120>, 1=<0 90 120>, 2=<0 0 0>, 3=<0 0 0>, 4=<0 0 0>, Index=&SpawnGroup(), Num=::NewEnumerator3)
    Sprite Size=&MergeHelper_Vector2Switch(0=<0 0>, 1=<500 500>, 2=<0 0>, 3=<0 0>, 4=<0 0>, Index=&SpawnGroup(), Num=::NewEnumerator3)
    Sprite Size Mode=::NewEnumerator4
    UsePositionOffset=true
    Write Material Random=true
   )
   &ShapeLocation(
    MaskSpawnGroup=false
    Non Uniform Scale=<1 0 1>
    Override Seed=false
    Sphere Radius=&MakeCustomFloatFromBool(
     Bool=&SetBoolByIntComparison(A=$Particles.MeshIndex, B=12, Comparison Type=::NewEnumerator6)
     True Float=5
     False Float=0
    )
    UseOwnerScale=true
   )
   &AddVelocity(
    Cone Angle=&MakeCustomFloatFromBool(
     Bool=&SetBoolByIntComparison(A=$Particles.MeshIndex, B=8, Comparison Type=::NewEnumerator6)
     True Float=55
     False Float=&MakeCustomFloatFromBool(
      Bool=&Boolean_AndOperation(
       Bool1=&SetBoolByIntComparison(A=$Particles.MeshIndex, B=8, Comparison Type=::NewEnumerator2)
       Bool2=&SetBoolByIntComparison(A=$Particles.MeshIndex, B=12, Comparison Type=::NewEnumerator6)
      )
      True Float=90
      False Float=0
     )
    )
    Cone Axis=<0 1 0>
    Module.UseConeFalloff=false
    Module.UseVelDistribution=false
    Override Seed=false
    Velocity Mode=::NewEnumerator2
    Velocity Speed=&MergeHelper_FloatSwitch(
     0=&MakeCustomFloatFromBool(
      Bool=&SetBoolByIntComparison(A=$Particles.MeshIndex, B=8, Comparison Type=::NewEnumerator6)
      True Float=&RandomRangeFloat(Minimum=100, Maximum=300)
      False Float=&MakeCustomFloatFromBool(
       Bool=&Boolean_AndOperation(
        Bool1=&SetBoolByIntComparison(A=$Particles.MeshIndex, B=8, Comparison Type=::NewEnumerator2)
        Bool2=&SetBoolByIntComparison(A=$Particles.MeshIndex, B=12, Comparison Type=::NewEnumerator6)
       )
       True Float=&RandomRangeFloat(Minimum=400, Maximum=600)
       False Float=0
      )
     )
     1=0
     2=0
     3=0
     4=0
     Index=&SpawnGroup()
     Num=::NewEnumerator3
    )
   ) {enabled=false}
   &AddVelocity(
    Cone Angle=&MakeCustomFloatFromBool(
     Bool=&SetBoolByIntComparison(A=$Particles.MeshIndex, B=8, Comparison Type=::NewEnumerator6)
     True Float=55
     False Float=&MakeCustomFloatFromBool(
      Bool=&Boolean_AndOperation(
       Bool1=&SetBoolByIntComparison(A=$Particles.MeshIndex, B=8, Comparison Type=::NewEnumerator2)
       Bool2=&SetBoolByIntComparison(A=$Particles.MeshIndex, B=12, Comparison Type=::NewEnumerator6)
      )
      True Float=90
      False Float=0
     )
    )
    Cone Axis=<0 1 0>
    Module.UseConeFalloff=false
    Module.UseVelDistribution=false
    Override Seed=false
    Velocity Mode=::NewEnumerator2
    Velocity Speed=&MergeHelper_FloatSwitch(
     0=&Multiply_Float(
      A=$System.SpeedMult
      B=&MakeCustomFloatFromBool(
       Bool=&Boolean_OrOperation(
        Bool1=&SetBoolByIntComparison(A=$Particles.MeshIndex, B=0, Comparison Type=::NewEnumerator4)
        Bool2=&Boolean_OrOperation(
         Bool1=&SetBoolByIntComparison(A=$Particles.MeshIndex, B=1, Comparison Type=::NewEnumerator4)
         Bool2=&Boolean_OrOperation(
          Bool1=&SetBoolByIntComparison(A=$Particles.MeshIndex, B=7, Comparison Type=::NewEnumerator4)
          Bool2=&Boolean_OrOperation(
           Bool1=&SetBoolByIntComparison(A=$Particles.MeshIndex, B=8, Comparison Type=::NewEnumerator4)
           Bool2=&SetBoolByIntComparison(A=$Particles.MeshIndex, B=10, Comparison Type=::NewEnumerator4)
          )
         )
        )
       )
       True Float=&UniformAOrBFloat(
        A=&RandomRangeFloat(Minimum=400, Maximum=600)
        B=680
        Distribution Weight=75
        Evaluation Type=::NewEnumerator0
        Recalculate Random Each Loop=true
        Override Seed=false
       )
       False Float=&MakeCustomFloatFromBool(
        Bool=&Boolean_AndOperation(
         Bool1=&SetBoolByIntComparison(A=$Particles.MeshIndex, B=12, Comparison Type=::NewEnumerator6)
         Bool2=&Boolean_OrOperation(
          Bool1=&SetBoolByIntComparison(A=$Particles.MeshIndex, B=1, Comparison Type=::NewEnumerator5)
          Bool2=&Boolean_OrOperation(
           Bool1=&SetBoolByIntComparison(A=$Particles.MeshIndex, B=7, Comparison Type=::NewEnumerator5)
           Bool2=&Boolean_OrOperation(
            Bool1=&SetBoolByIntComparison(A=$Particles.MeshIndex, B=8, Comparison Type=::NewEnumerator5)
            Bool2=&SetBoolByIntComparison(A=$Particles.MeshIndex, B=10, Comparison Type=::NewEnumerator5)
           )
          )
         )
        )
        True Float=&RandomRangeFloat(Minimum=100, Maximum=300)
        False Float=0
       )
      )
     )
     1=0
     2=0
     3=0
     4=0
     Index=&SpawnGroup()
     Num=::NewEnumerator3
    )
   ) {module=AddVelocity002}
   &AddVelocity(Velocity=&MergeHelper_VectorSwitch(
     0=&Multiply_VectorByFloat(
      Vector=&MakeCustomVectorFromBool(
       Bool=&SetBoolByIntComparison(A=$Particles.MeshIndex, B=12, Comparison Type=::NewEnumerator6)
       True Vector=&RandomRangeVector(Minimum=<0 0 50>, Maximum=<0 0 100>)
       False Vector=<0 0 0>
      )
      Float=$System.SpeedMult
     )
     1=<0 0 0>
     2=<0 0 0>
     3=<0 0 0>
     4=<0 0 0>
     Index=&SpawnGroup()
     Num=::NewEnumerator3
    )) {module=AddVelocity001}
   /SetVariables(Particles.Flash DynParam:Vector4f=&MakeVector4(X=&RandomRangeFloat(Minimum=0.4, Maximum=0.6), Y=0, Z=0, W=0)) {isSetParameters=true}
   &CameraOffset(Camera Offset Amount=&MergeHelper_FloatSwitch(0=0, 1=1, 2=0, 3=0, 4=0, Index=&SpawnGroup(), Num=::NewEnumerator3))
  ]
  particleUpdateScript [
   &ParticleState()
   &UpdateMeshOrientation(
    Rotation Rate=&MakeCustomFloatFromBool(
     Bool=&SetBoolByIntComparison(A=$Particles.MeshIndex, B=12, Comparison Type=::NewEnumerator0)
     True Float=&MakeCustomFloatFromBool(
      Bool=&SetBoolByFloatComparison(A=&VectorLength(Vector=$Particles.Velocity), B=50)
      True Float=&UniformAOrBFloat(
       A=&RandomRangeFloat(Minimum=-1, Maximum=1)
       B=&RandomRangeFloat(Minimum=2, Maximum=4)
       Distribution Weight=0.2
       Evaluation Type=::NewEnumerator0
       Recalculate Random Each Loop=true
       Override Seed=false
      )
      False Float=0
     )
     False Float=0
    )
    Rotation Vector=&MakeCustomVectorFromBool(
     Bool=&SetBoolByIntComparison(A=$Particles.MeshIndex, B=12, Comparison Type=::NewEnumerator0)
     True Vector=&RandomRangeVector(Minimum=<-1 -1 -1>, Maximum=<1 1 1>)
     False Vector=<0 0 0>
    )
   ) {module=UpdateMeshOrientation001}
   &UpdateMeshOrientation(
    Facing Direction=&MakeCustomVectorFromBool(
     Bool=&SetBoolByIntComparison(A=$Particles.MeshIndex, B=8, Comparison Type=::NewEnumerator6)
     True Vector=&MergeHelper_VectorSwitch(
      0=<0.5 1 1>
      1=<-0.5 1 -0.5>
      2=<0.5 -1 -1>
      3=<1 3 0>
      4=<-0.7 1 0>
      5=<0.5 1 0>
      6=<0.75 1 0>
      Index=$Particles.MeshIndex
      Num=::NewEnumerator5
     )
     False Vector=&MakeCustomVectorFromBool(
      Bool=&Boolean_AndOperation(
       Bool1=&SetBoolByIntComparison(A=$Particles.MeshIndex, B=8, Comparison Type=::NewEnumerator2)
       Bool2=&SetBoolByIntComparison(A=$Particles.MeshIndex, B=12, Comparison Type=::NewEnumerator6)
      )
      True Vector=&Lerp_Vector(
       A=&RandomRangeVector(Minimum=<-2 1 1>, Maximum=<2 1 -1>)
       B=&RandomRangeVector(Minimum=<0 -1 0>, Maximum=<0 1 1>)
       Alpha=&FloatFromCurve(FloatCurve=DataInterface() {propertyValues {Curve {defaultValue=1, keys [{tangentMode=RCTM_User, time=0.1756, value=0}, {tangentMode=RCTM_User, time=0.326, value=1}]}}})
      )
      False Vector=<1 0 0>
     )
    )
    Facing Interpolation=false
    Orientation Method=::NewEnumerator1
    Override Quat=false
    TransformFacing=false
    TransformUp=false
   )
   &GravityForce(Gravity=&MergeHelper_VectorSwitch(
     0=&MakeVector(
      X=0
      Y=0
      Z=&MakeCustomFloatFromBool(
       Bool=&SetBoolByIntComparison(A=$Particles.MeshIndex, B=8, Comparison Type=::NewEnumerator6)
       True Float=&RandomRangeFloat(Minimum=-800, Maximum=-1200)
       False Float=&MakeCustomFloatFromBool(
        Bool=&Boolean_AndOperation(
         Bool1=&SetBoolByIntComparison(A=$Particles.MeshIndex, B=8, Comparison Type=::NewEnumerator2)
         Bool2=&SetBoolByIntComparison(A=$Particles.MeshIndex, B=12, Comparison Type=::NewEnumerator6)
        )
        True Float=&RandomRangeFloat(Minimum=-800, Maximum=-900)
        False Float=&MakeCustomFloatFromBool(
         Bool=&Boolean_OrOperation(
          Bool1=&SetBoolByIntComparison(A=$Particles.MeshIndex, B=13, Comparison Type=::NewEnumerator4)
          Bool2=&Boolean_OrOperation(
           Bool1=&SetBoolByIntComparison(A=$Particles.MeshIndex, B=17, Comparison Type=::NewEnumerator4)
           Bool2=&SetBoolByIntComparison(A=$Particles.MeshIndex, B=20, Comparison Type=::NewEnumerator4)
          )
         )
         True Float=&UniformAOrBFloat(
          A=&FloatFromCurve(
           FloatCurve=DataInterface() {propertyValues {Curve {defaultValue=1, keys [{tangentMode=RCTM_User, time=0.55, value=0}, {tangentMode=RCTM_User, time=0.75, value=1}]}}}
           Scale Curve=-5000
          )
          B=&FloatFromCurve(
           FloatCurve=DataInterface() {propertyValues {Curve {defaultValue=1, keys [{tangentMode=RCTM_User, time=0.75, value=0}, {tangentMode=RCTM_User, time=0.85, value=1}]}}}
           Scale Curve=-6000
          )
          Evaluation Type=::NewEnumerator0
          Recalculate Random Each Loop=true
          Override Seed=false
         )
         False Float=&MakeCustomFloatFromBool(
          Bool=&Boolean_OrOperation(
           Bool1=&SetBoolByIntComparison(A=$Particles.MeshIndex, B=12, Comparison Type=::NewEnumerator4)
           Bool2=&Boolean_OrOperation(
            Bool1=&SetBoolByIntComparison(A=$Particles.MeshIndex, B=15, Comparison Type=::NewEnumerator4)
            Bool2=&SetBoolByIntComparison(A=$Particles.MeshIndex, B=19, Comparison Type=::NewEnumerator4)
           )
          )
          True Float=&UniformAOrBFloat(
           A=&FloatFromCurve(
            FloatCurve=DataInterface() {propertyValues {Curve {defaultValue=1, keys [{tangentMode=RCTM_User, time=0.75, value=0}, {tangentMode=RCTM_User, time=0.85, value=1}]}}}
            Scale Curve=-4000
           )
           B=&FloatFromCurve(
            FloatCurve=DataInterface() {propertyValues {Curve {defaultValue=1, keys [{tangentMode=RCTM_User, time=0.65, value=0}, {tangentMode=RCTM_User, time=0.85, value=1}]}}}
            Scale Curve=-5000
           )
           Evaluation Type=::NewEnumerator0
           Recalculate Random Each Loop=true
           Override Seed=false
          )
          False Float=&UniformAOrBFloat(
           A=&FloatFromCurve(
            FloatCurve=DataInterface() {propertyValues {Curve {defaultValue=1, keys [{tangentMode=RCTM_User, time=0.6, value=0}, {tangentMode=RCTM_User, time=0.7, value=1}]}}}
            Scale Curve=-5000
           )
           B=&FloatFromCurve(
            FloatCurve=DataInterface() {propertyValues {Curve {defaultValue=1, keys [{tangentMode=RCTM_User, time=0.35, value=0}, {tangentMode=RCTM_User, time=0.45, value=1}]}}}
            Scale Curve=-6000
           )
           Evaluation Type=::NewEnumerator0
           Recalculate Random Each Loop=true
           Override Seed=false
          )
         )
        )
       )
      )
     )
     1=<0 0 0>
     2=<0 0 0>
     3=<0 0 0>
     4=<0 0 0>
     Index=&SpawnGroup()
     Num=::NewEnumerator3
    ))
   &CurlNoiseForce(
    Bias Noise=false
    Noise Frequency=&MergeHelper_FloatSwitch(0=10, 1=0, 2=0, 3=0, 4=0, Index=&SpawnGroup(), Num=::NewEnumerator3)
    Noise Strength=&MergeHelper_FloatSwitch(
     0=&MakeCustomFloatFromBool(
      Bool=&SetBoolByIntComparison(A=$Particles.MeshIndex, B=12, Comparison Type=::NewEnumerator6)
      True Float=100
      False Float=0
     )
     1=0
     2=0
     3=0
     4=0
     Index=&SpawnGroup()
     Num=::NewEnumerator3
    )
    Pan Noise=false
   )
   &MeshRotationForce(
    Coordinate Space:ENiagaraMeshTransforms=::NewEnumerator3
    Rotation=&MakeCustomVectorFromBool(
     Bool=&SetBoolByIntComparison(A=$Particles.MeshIndex, B=12, Comparison Type=::NewEnumerator6)
     True Vector=&MakeCustomVectorFromBool(
      Bool=&SetBoolByIntComparison(A=$Particles.MeshIndex, B=8, Comparison Type=::NewEnumerator6)
      True Vector=&RandomRangeVector(Minimum=<-20 -20 -20>, Maximum=<20 20 20>)
      False Vector=&RandomRangeVector(Minimum=<-70 -70 -70>, Maximum=<70 70 70>)
     )
     False Vector=&VectorFromFloat(Value=&FloatFromCurve(
       FloatCurve=DataInterface() {propertyValues {Curve {defaultValue=1, keys [<at 0.8 0 RCIM_Cubic>,<at 1.2 1 RCIM_Cubic>]}}}
       Scale Curve=&RandomRangeFloat(Minimum=-40, Maximum=40)
      ))
    )
   )
   &AccelerationForce(Acceleration=&MergeHelper_VectorSwitch(
     0=&Multiply_VectorByFloat(
      Vector=&MakeCustomVectorFromBool(
       Bool=&SetBoolByIntComparison(A=$Particles.MeshIndex, B=12, Comparison Type=::NewEnumerator6)
       True Vector=<0 5 0>
       False Vector=<0 0 0>
      )
      Float=$System.SpeedMult
     )
     1=<0 0 0>
     2=<0 0 0>
     3=<0 0 0>
     4=<0 0 0>
     Index=&SpawnGroup()
     Num=::NewEnumerator3
    ))
   &Drag(
    Drag=&MakeCustomFloatFromBool(
     Bool=&SetBoolByIntComparison(A=$Particles.MeshIndex, B=12, Comparison Type=::NewEnumerator6)
     True Float=2
     False Float=0
    )
    Use Linear Drag=true
    Use Rotational Drag=true
   )
   &SolveForcesAndVelocity(Clamp Velocity=false, Limit Acceleration=false)
   &ScaleColor(
    Scale Alpha=&MergeHelper_FloatSwitch(
     0=1
     1=&FloatFromCurve(FloatCurve=DataInterface() {propertyValues {Curve {keys [
          {interpMode=RCIM_Cubic, tangentMode=RCTM_User, time=0, value=0}
          {interpMode=RCIM_Cubic, tangentMode=RCTM_User, time=0.2, value=1}
          {interpMode=RCIM_Cubic, tangentMode=RCTM_User, time=1, value=0}
         ]}}})
     2=0
     3=0
     4=0
     Index=&SpawnGroup()
     Num=::NewEnumerator3
    )
    Scale RGB=&MergeHelper_VectorSwitch(
     0=<1 1 1>
     1=&Multiply_VectorByFloat(
      Vector=&MakeVectorFromLinearColorRGB(Color=&ColorFromCurve(ColorCurve=DataInterface() {propertyValues {
          AlphaCurve {keys []}
          BlueCurve {keys [<at 0 2 RCIM_Cubic>,<at 0.2616 1 RCIM_Cubic>]}
          GreenCurve {keys [<at 0 2 RCIM_Cubic>,<at 0.2616 1 RCIM_Cubic>]}
          RedCurve {keys [<at 0 2 RCIM_Cubic>,<at 0.2616 1 RCIM_Cubic>]}
         }}))
      Float=5
     )
     2=<0 0 0>
     3=<0 0 0>
     4=<0 0 0>
     Index=&SpawnGroup()
     Num=::NewEnumerator3
    )
    ScaleA=true
    ScaleRGB=true
   )
   &ScaleSpriteSize(Scale Factor=&MergeHelper_Vector2Switch(
     0=<0 0>
     1=&Vector2DFromFloat(Value=&FloatFromCurve(
       FloatCurve=DataInterface() {propertyValues {Curve {
          defaultValue=0
          keys [{interpMode=RCIM_Cubic, tangentMode=RCTM_User, time=0, value=0}, {interpMode=RCIM_Cubic, tangentMode=RCTM_User, time=1, value=1}]
         }}}
       Scale Curve=&FN_FrontEndFloatModifier(Gameplay=1.3, Front End=0.6)
      ))
     2=<0 0>
     3=<0 0>
     4=<0 0>
     Index=&SpawnGroup()
     Num=::NewEnumerator3
    ))
  ]
  rendererClasses [&NiagaraMeshRendererProperties, &NiagaraSpriteRendererProperties]
  renderers [
   {rendererClass=&NiagaraMeshRendererProperties, rendererIndex=0}
   {rendererClass=&NiagaraSpriteRendererProperties, rendererIndex=1}
  ]
 }
 Glows {
  emitterUpdateScript [
   &EmitterState(Life Cycle Mode=::NewEnumerator1, Loop Behavior=::NewEnumerator1, Loop Delay=$System.FireDelay, Loop Duration=2, UseLoopDelay=true)
   &SpawnBurst_Instantaneous(Spawn Time=0.1, Use Spawn Probability=false, UseLoopCountLimit=false)
   &SpawnBurst_Instantaneous(Spawn Group=1, Spawn Time=0.1, Use Spawn Probability=false, UseLoopCountLimit=false) {module=SpawnBurst_Instantaneous001}
  ]
  particleSpawnScript [
   /SetVariables(Particles.VisibilityTag=&SpawnGroup()) {isSetParameters=true}
   &InitializeParticle(
    Color Mode=::NewEnumerator0
    Lifetime=0.7
    Position Offset=<0 -40 90>
    Sprite Size Mode=::NewEnumerator3
    Uniform Sprite Size=200
    UsePositionOffset=true
    Write Material Random=true
   )
   &SpriteFacingAndAlignment(Sprite Facing=<0 1 0>, WriteAlignment=false, WriteFacing=true)
  ]
  particleUpdateScript [
   &ParticleState()
   /SetVariables(Particles.Squeeze:Vector2f=&MakeVector2D(
     X=&MergeHelper_FloatSwitch(
      0=&RemapRange(
       Input Value To Remap=&Abs_Float(Float:NiagaraFloat=&DotProduct(
         Vector A=&ReturnCameraVector(Camera Vectors=::NewEnumerator4, Coordinate Space=::Simulation, Origin Position=$Particles.Position)
         Vector B=$Particles.SpriteFacing
        ))
       Input Min=0.2
       Input Max=0.9
       Output Min=0.7
       Output Max=1
      )
      1=1
      2=?
      3=?
      Index=&SpawnGroup()
      Num=::NewEnumerator0
     )
     Y=1
    )) {isSetParameters=true}
   /SetVariables(
    Particles.GlowColorA:LinearColor=&MultiplyLinearColor(
     Linear Color=<rgba 1 0.8135 0.2344 1>
     Scale Factor=&MakeVector4_FromLinearColor(LinearColor=&MakeLinearColorFromVectorAndFloat(
       "Vector (RGB)"=&MakeVectorFromLinearColorRGB(Color=&ColorFromCurve(ColorCurve=DataInterface() {propertyValues {
           AlphaCurve {keys []}
           BlueCurve {keys [<at 0.6176 1 RCIM_Cubic>,<at 1 0.3156 RCIM_Cubic>]}
           GreenCurve {keys [<at 0.6176 1 RCIM_Cubic>,<at 1 0.3156 RCIM_Cubic>]}
           RedCurve {keys [<at 0.6176 1 RCIM_Cubic>,<at 1 0.3156 RCIM_Cubic>]}
          }}))
       "Float (Alpha)"=&FloatFromCurve(
        FloatCurve=DataInterface() {propertyValues {Curve {
           defaultValue=1
           keys [
            {arriveTangent=9.0234, interpMode=RCIM_Cubic, leaveTangent=9.0234, tangentMode=RCTM_User, time=0, value=0.3}
            {arriveTangent=0.312, interpMode=RCIM_Cubic, leaveTangent=0.312, tangentMode=RCTM_User, time=0.1591, value=1}
            {interpMode=RCIM_Cubic, tangentMode=RCTM_User, time=0.4, value=1}
            {arriveTangent=-0.0514, tangentMode=RCTM_User, time=1, value=0}
           ]
          }}}
        Scale Curve=0.5
       )
      ))
    )
    Particles.GlowColorB:LinearColor=&MultiplyLinearColor(
     Linear Color=<rgba 1.4609 0.1567 0 1>
     Scale Factor=&MakeVector4_FromLinearColor(LinearColor=&MakeLinearColorFromVectorAndFloat(
       "Vector (RGB)"=&MakeVectorFromLinearColorRGB(Color=&ColorFromCurve(ColorCurve=DataInterface() {propertyValues {
           AlphaCurve {keys []}
           BlueCurve {keys [<at 0.6176 1 RCIM_Cubic>,<at 1 0.3156 RCIM_Cubic>]}
           GreenCurve {keys [<at 0.6176 1 RCIM_Cubic>,<at 1 0.3156 RCIM_Cubic>]}
           RedCurve {keys [<at 0.6176 1 RCIM_Cubic>,<at 1 0.3156 RCIM_Cubic>]}
          }}))
       "Float (Alpha)"=&FloatFromCurve(
        FloatCurve=DataInterface() {propertyValues {Curve {
           defaultValue=1
           keys [
            {arriveTangent=9.0234, interpMode=RCIM_Cubic, leaveTangent=9.0234, tangentMode=RCTM_User, time=0, value=0.3}
            {arriveTangent=0.312, interpMode=RCIM_Cubic, leaveTangent=0.312, tangentMode=RCTM_User, time=0.1591, value=1}
            {interpMode=RCIM_Cubic, tangentMode=RCTM_User, time=0.4, value=1}
            {arriveTangent=-0.0514, tangentMode=RCTM_User, time=1, value=0}
           ]
          }}}
        Scale Curve=0.5
       )
      ))
    )
    Particles.GlowSizeA:Vector2f=&Multiply_Vector2D(A=<185 250>, B=$Particles.Squeeze)
    Particles.GlowSizeB:Vector2f=&Multiply_Vector2D(A=<330 270>, B=$Particles.Squeeze)
   ) {isSetParameters=true}
  ]
  rendererClasses [&NiagaraSpriteRendererProperties]
  renderers [
   {rendererClass=&NiagaraSpriteRendererProperties}
   {rendererClass=&NiagaraSpriteRendererProperties, rendererIndex=1}
  ]
 }
 MeshSoftFlame {
  emitterUpdateScript [
   &EmitterState(Life Cycle Mode=::NewEnumerator1, Loop Behavior=::NewEnumerator1, Loop Delay=$System.FireDelay, Loop Duration=2, UseLoopDelay=true)
   &SpawnBurst_Instantaneous(Spawn Count=11, Use Spawn Probability=false, UseLoopCountLimit=false)
   /SetVariables(
    Emitter.Base Color 1:LinearColor=&ColorFromCurve(
     ColorCurve=DataInterface() {propertyValues {
       AlphaCurve {keys [<at 0 1 RCIM_Cubic>]}
       BlueCurve {keys [
         <at 0.0071 0.0233>
         <at 0.0608 0>
         <at 0.1709 0.0838>
         {arriveTangent=-0.1084, interpMode=RCIM_Cubic, leaveTangent=-0.1084, time=0.5153, value=0}
         <at 0.9436 0>
        ]}
       GreenCurve {keys [
         <at 0.0071 0.479>
         <at 0.0608 1.659>
         <at 0.1709 0.2514>
         {arriveTangent=-0.2427, interpMode=RCIM_Cubic, leaveTangent=-0.2427, time=0.5153, value=0.3214}
         <at 0.9436 0.0638>
        ]}
       RedCurve {keys [
         <at 0.0071 9>
         <at 0.0608 15>
         <at 0.1709 9>
         {arriveTangent=14.2373, interpMode=RCIM_Cubic, leaveTangent=14.2373, time=0.5153, value=9}
         <at 0.9436 20>
        ]}
      }}
     CurveIndex=$Emitter.NormalizedLoopAge
    )
    Emitter.Base Color 2:LinearColor=<rgba 20 9.5687 1.1274 1>
   ) {isSetParameters=true}
  ]
  particleSpawnScript [
   &PartitionParticles(Number Of Partitions=7)
   /SetVariables(Particles.VectorScale:Vector3f=&Multiply_VectorByFloat(Vector=<0.7 0.7 1>, Float=&RandomRangeFloat(Minimum=1, Maximum=4))) {isSetParameters=true}
   &InitializeParticle(
    Color=<rgba 1 1 1 0.2>
    Lifetime=&UniformAOrBFloat(
     A=&RandomRangeFloat(Minimum=0.3, Maximum=0.45)
     B=&RandomRangeFloat(Minimum=0.45, Maximum=0.55)
     Distribution Weight=0.65
     Evaluation Type=::NewEnumerator0
     Recalculate Random Each Loop=true
     Override Seed=false
    )
    Mesh Scale=&Multiply_Vector(
     A=&Multiply_Vector(A=&RandomRangeVector(Minimum=<1 1 1.5>, Maximum=<1 1 1>), B=$Particles.VectorScale)
     B=&VectorFromFloat(Value=$Particles.SizeScle)
    )
    Mesh Scale Mode=::NewEnumerator4
    UsePositionOffset=false
    Write Material Random=true
   ) {module=InitializeParticle001}
   &InitialMeshOrientation(
    Rotation=&RandomRangeVector(Minimum=<0 0 0>, Maximum=<0 0 1>)
    Use Orientation Vector=true
    Use Rotation Vector=true
   )
   &ShapeLocation(
    Box Size=&MergeHelper_VectorSwitch(
     0=<35 0 65>
     1=<22 0 65>
     2=<22 0 65>
     3=<22 0 65>
     4=<22 0 65>
     5=<22 0 65>
     6=<22 0 65>
     Index=$Particles.PartitionParticles.Partition
     Num=::NewEnumerator5
    )
    MaskSpawnGroup=false
    Offset=&MergeHelper_VectorSwitch(
     0=<0 -40 60>
     1=<-20 -40 60>
     2=<20 -40 60>
     3=<-40 -40 60>
     4=<40 -40 60>
     5=<-60 -40 60>
     6=<60 -40 60>
     Index=$Particles.PartitionParticles.Partition
     Num=::NewEnumerator5
    )
    Offset Mode=::NewEnumerator0
    Override Seed=false
    Shape Primitive=::NewEnumerator2
    Surface Only=false
    UseOwnerScale=true
   ) {module=ShapeLocation002}
  ]
  particleUpdateScript [
   /SetVariables(Particles.Mesh_Color2:LinearColor=&MakeLinearColorFromVectorAndFloat(
     "Vector (RGB)"=&MakeVectorFromLinearColorRGB(Color=&MultiplyLinearColor(
       Linear Color=&MergeHelper_ColorSwitch(
        0=&RandomRangeLinearColor(Minimum=<rgba 10 10 10 1>, Maximum=<rgba 10 10 10 1>)
        1=&RandomRangeLinearColor(Minimum=<rgba 1 0.7625 0.5678 1>, Maximum=<rgba 1 0.853 0.7344 1>)
        2=&RandomRangeLinearColor(Minimum=<rgba 1 0.7625 0.5678 1>, Maximum=<rgba 1 0.853 0.7344 1>)
        3=&RandomRangeLinearColor(Minimum=<rgba 0.699 0.5573 0.1098 1>, Maximum=<rgba 0.524 0.3098 0.1213 1>)
        4=&RandomRangeLinearColor(Minimum=<rgba 0.699 0.5573 0.1098 1>, Maximum=<rgba 0.524 0.3098 0.1213 1>)
        5=&RandomRangeLinearColor(Minimum=<rgba 0.524 0.3098 0.1213 1>, Maximum=<rgba 0.1448 0.0903 0 1>)
        6=&RandomRangeLinearColor(Minimum=<rgba 0.524 0.3098 0.1213 1>, Maximum=<rgba 0.1448 0.0903 0 1>)
        Index=$Particles.PartitionParticles.Partition
        Num=::NewEnumerator5
       )
       Scale Factor=&MakeVector4_FromFloat(Float=&FN_FrontEndFloatModifier(Gameplay=1, Front End=1.3))
      ))
     "Float (Alpha)"=&Multiply_Float(
      A=0.5
      B=&FloatFromCurve(FloatCurve=DataInterface() {propertyValues {Curve {keys [<at 0 0>,<at 0.1269 1>,<at 0.4657 0.4118>,<at 1 0>]}}})
     )
    )) {isSetParameters=true}
   /SetVariables(Particles.Mesh_Size2:Vector3f=&Multiply_VectorByFloat(
     Vector=&Multiply_Vector(A=<1 0.9 1.2>, B=$Particles.VectorScale)
     Float=&FloatFromCurve(FloatCurve=DataInterface() {propertyValues {Curve {
         defaultValue=0
         keys [
          {interpMode=RCIM_Cubic, tangentMode=RCTM_User, time=0, value=0}
          {interpMode=RCIM_Cubic, tangentMode=RCTM_User, time=0.5083, value=1}
          {arriveTangent=0.2646, interpMode=RCIM_Cubic, leaveTangent=0.2646, tangentMode=RCTM_User, time=1.0033, value=1.2524}
         ]
        }}})
    )) {isSetParameters=true}
   &ParticleState()
   &InterpolateOverTime(Rate Of Change=20, Target Vector Value:NiagaraPosition=$Particles.Position, Variable Type=::NewEnumerator2) {module=InterpolateOverTime001}
   /SetVariables(Particles.Interpolate Dynamic=&MakeFloat4FromVectorAndFloat("Vector (RGB)"=$Particles.InterpolateOverTime001.InterpolateOverTime_V3_.Moving Average, "Float (Alpha)"=0.8)) {isSetParameters=true}
   &SolveForcesAndVelocity(Clamp Velocity=false, Limit Acceleration=false)
   &ScaleColor(
    Scale Alpha=&FloatFromCurve(
     FloatCurve=DataInterface() {propertyValues {Curve {keys [
         <at 0.0756 0>
         <at 0.1339 1>
         {arriveTangent=-1.2231, interpMode=RCIM_Cubic, leaveTangent=-1.2231, time=0.4105, value=1}
         <at 0.9611 -0.0118 RCIM_Cubic>
        ]}}}
     Scale Curve=1.25
    )
    Scale RGB=&Multiply_VectorByFloat(
     Vector=<1 1 1>
     Float=&FloatFromCurve(FloatCurve=DataInterface() {propertyValues {Curve {
         defaultValue=1
         keys [
          {interpMode=RCIM_Cubic, tangentMode=RCTM_User, time=0, value=1}
          {interpMode=RCIM_Cubic, tangentMode=RCTM_User, time=0.4262, value=5.8588}
          <at 1 1 RCIM_Cubic>
         ]
        }}})
    )
    ScaleA=true
    ScaleRGB=true
   )
  ]
  rendererClasses [&NiagaraMeshRendererProperties]
  renderers [{rendererClass=&NiagaraMeshRendererProperties, rendererIndex=0}]
 }
 MiniGlass {
  emitterUpdateScript [
   &EmitterState(Life Cycle Mode=::NewEnumerator1, Loop Behavior=::NewEnumerator1, Loop Duration=1, UseLoopDelay=false)
   &SpawnBurst_Instantaneous(Spawn Count=15, Use Spawn Probability=false, UseLoopCountLimit=false)
   /SetVariables(Emitter.Opacity:NiagaraFloat=&FloatFromCurve(
     FloatCurve=DataInterface() {propertyValues {Curve {keys [
         {arriveTangent=-4.7558, interpMode=RCIM_Cubic, leaveTangent=-4.7558, tangentMode=RCTM_User, time=1, value=0.8}
         {arriveTangent=-1.0407, interpMode=RCIM_Cubic, leaveTangent=-1.0407, tangentMode=RCTM_User, time=1.2, value=0.3462}
         {arriveTangent=-0.7988, interpMode=RCIM_Cubic, leaveTangent=-0.7988, tangentMode=RCTM_User, time=1.6, value=0}
        ]}}}
     CurveIndex=$Emitter.Age
    )) {isSetParameters=true}
  ]
  particleSpawnScript [
   &InitializeParticle(
    Lifetime Max=0.35
    Lifetime Min=0.15
    Lifetime Mode=::NewEnumerator1
    Mass Max=2
    Mass Mode=::NewEnumerator3
    Mesh Scale Mode=::NewEnumerator3
    Mesh Uniform Scale=&RandomRangeFloat(Minimum=0.5, Maximum=0.85)
    Position Offset=&RandomRangeVector(Minimum=<0 90 120>, Maximum=<0 110 120>)
    UsePositionOffset=true
    Write Material Random=true
   )
   &AddVelocity(
    Cone Angle=50
    Cone Axis=<0 1 0>
    Distribution Along Cone Axis=0.2
    Module.UseConeFalloff=false
    Module.UseVelDistribution=true
    Override Seed=false
    Velocity Mode=::NewEnumerator2
    Velocity Speed=&Multiply_Float(A=&RandomRangeFloat(Minimum=250, Maximum=750), B=$System.SpeedMult)
   )
   &InitialMeshOrientation(Use Rotation Vector=false)
  ]
  particleUpdateScript [
   &ParticleState()
   &GravityForce()
   &Drag(Drag=&RandomRangeFloat(Minimum=0.8, Maximum=1.2), Use Linear Drag=true, Use Rotational Drag=false)
   &MeshRotationForce(
    Coordinate Space:ENiagaraMeshTransforms=::NewEnumerator3
    Rotation=&MakeCustomVectorFromBool(
     Bool=&SetBoolByIntComparison(A=$Particles.MeshIndex, B=12, Comparison Type=::NewEnumerator6)
     True Vector=&MakeCustomVectorFromBool(
      Bool=&SetBoolByIntComparison(A=$Particles.MeshIndex, B=8, Comparison Type=::NewEnumerator6)
      True Vector=&RandomRangeVector(Minimum=<-70 -70 -70>, Maximum=<70 70 70>)
      False Vector=&RandomRangeVector(Minimum=<-20 -20 -20>, Maximum=<20 20 20>)
     )
     False Vector=&VectorFromFloat(Value=&FloatFromCurve(
       FloatCurve=DataInterface() {propertyValues {Curve {defaultValue=1, keys [<at 1.2 0 RCIM_Cubic>,<at 1.6 1 RCIM_Cubic>]}}}
       Scale Curve=&RandomRangeFloat(Minimum=-7, Maximum=7)
      ))
    )
   )
   &ScaleColor(
    Scale Alpha=&FloatFromCurve(FloatCurve=DataInterface() {propertyValues {Curve {keys [<at -0.0046 1.0113 RCIM_Cubic>,<at 1.0011 0.002 RCIM_Cubic>]}}})
    ScaleA=true
    ScaleRGB=true
   )
   &SolveForcesAndVelocity(Clamp Velocity=false, Limit Acceleration=false)
   &ScaleSpriteSizeBySpeed(
    Max Scale Factor=<1.2 3.3>
    Min Scale Factor=<0.75 0.3>
    Sample Scale Factor By Curve=true
    Scale Factor Curve=DataInterface() {propertyValues {Curve {keys [
        {arriveTangent=2.865, interpMode=RCIM_Cubic, time=0.1247, value=-0.0006}
        {arriveTangent=4.1522, interpMode=RCIM_Cubic, leaveTangent=4.1522, tangentMode=RCTM_User, time=0.3948, value=0.4893}
        <at 0.9958 0.9974 RCIM_Cubic>
       ]}}}
    Velocity Threshold=1500
   )
  ]
  rendererClasses [&NiagaraMeshRendererProperties]
  renderers [{rendererClass=&NiagaraMeshRendererProperties, rendererIndex=0}]
 }
 Rose {
  emitterUpdateScript [
   &EmitterState(Life Cycle Mode=::NewEnumerator1, Loop Behavior=::NewEnumerator1, Loop Delay=$System.FireDelay, Loop Duration=2, UseLoopDelay=true)
   &SpawnRate(
    SpawnRate=&Multiply_Float(
     A=&FloatFromCurve(
      FloatCurve=DataInterface() {propertyValues {Curve {keys [<at 0.439 0>,<at 0.44 1>,<at 1.0653 1>,<at 1.078 0>]}}}
      CurveIndex=$Emitter.NormalizedLoopAge
     )
     B=2
    )
    Use Spawn Probability=false
   )
   &SpawnBurst_Instantaneous(Spawn Count=3, Spawn Time=$System.FireDelay, Use Spawn Probability=false, UseLoopCountLimit=false)
  ]
  particleSpawnScript [
   &InitializeParticle(
    Color Mode=::NewEnumerator0
    Lifetime Max=0.6
    Lifetime Min=0.5
    Lifetime Mode=::NewEnumerator1
    Mesh Scale Max=<1.2 0.2 1.2>
    Mesh Scale Min=<0.3 0.3 0.3>
    Mesh Scale Mode=::NewEnumerator2
    UsePositionOffset=false
    Write Material Random=true
   )
   &InitialMeshOrientation(
    Mesh Orientation Mode=::NewEnumerator2
    Rotation=&RandomRangeVector(Minimum=<0 0 0>, Maximum=<1 1 1>)
    Use Rotation Vector=true
   )
   &ShapeLocation(
    Box Size=<25 5 10>
    MaskSpawnGroup=false
    Offset=<0 -40 50>
    Offset Mode=::NewEnumerator0
    Override Seed=false
    Shape Primitive=::NewEnumerator2
    Surface Only=false
    UseOwnerScale=true
   )
   &AddVelocity(
    Cone Angle=60
    Cone Axis=&RandomRangeVector(Minimum=<-0.5 0 1>, Maximum=<0.5 0 1>)
    Module.UseConeFalloff=false
    Module.UseVelDistribution=false
    Override Seed=false
    Velocity Mode=::NewEnumerator2
    Velocity Speed=&RandomRangeFloat(Minimum=150, Maximum=200)
   )
   &CameraOffset(Camera Offset Amount=&RandomRangeFloat(Minimum=15, Maximum=20))
  ]
  particleUpdateScript [
   &ParticleState()
   &UpdateMeshOrientation(Rotation Rate=&RandomRangeFloat(Minimum=0.8, Maximum=1.2), Rotation Vector=<1 0.2 0.4>)
   &CurlNoiseForce(Bias Noise=false, Noise Strength=60, Pan Noise=false)
   &ScaleColor(
    Scale Alpha=&FloatFromCurve(FloatCurve=DataInterface() {propertyValues {Curve {keys [<at 0.8 1>,<at 1 1>]}}})
    Scale RGB=&MakeVectorFromLinearColorRGB(Color=&RandomRangeLinearColor(Minimum=<rgba 3 1.4298 0.2134 1>, Maximum=<rgba 3 0.2626 0.0404 1>))
    ScaleA=true
    ScaleRGB=true
   )
   &Drag(
    Drag=&FloatFromCurve(
     FloatCurve=DataInterface() {propertyValues {Curve {keys [<at 0 0.2 RCIM_Cubic>, {arriveTangent=17.1177, interpMode=RCIM_Cubic, leaveTangent=17.1177, tangentMode=RCTM_User, time=0.4, value=3}]}}}
     Scale Curve=&RandomRangeFloat(Minimum=1, Maximum=1.2)
    )
    Use Linear Drag=true
    Use Rotational Drag=true
   )
   &GravityForce(
    Coordinate Space=::Local
    Gravity=&MakeVector(
     X=&RandomRangeFloat(Minimum=-35, Maximum=-120)
     Y=0
     Z=&FloatFromCurve(
      FloatCurve=DataInterface() {propertyValues {Curve {
         defaultValue=1
         keys [
          {arriveTangent=0.063, interpMode=RCIM_Cubic, leaveTangent=0.063, tangentMode=RCTM_User, time=0, value=0}
          {arriveTangent=4.2755, interpMode=RCIM_Cubic, leaveTangent=4.2755, tangentMode=RCTM_User, time=0.4, value=1}
         ]
        }}}
      Scale Curve=&RandomRangeFloat(Minimum=25, Maximum=225)
     )
    )
   )
   &SolveForcesAndVelocity(Clamp Velocity=false, Limit Acceleration=false)
   &DynamicMaterialParameters(
    Index 0 Param 1=&FloatFromCurve(
     FloatCurve=DataInterface() {propertyValues {Curve {keys [<at 0.3 1>,<at 1 -0.3>]}}}
     Scale Curve=&RandomRangeFloat(Minimum=0.8, Maximum=1)
    )
    Index 0 Param 2=&RandomRangeFloat(Minimum=0, Maximum=0.5)
    Index 0 Param 3=&RandomRangeFloat(Minimum=0, Maximum=0.5)
    Index 0 Param 4=&RandomRangeFloat(Minimum=0.5, Maximum=2)
    Module.Index0Float0Write=true
    Module.Index0Float1Write=true
    Module.Index0Float2Write=true
    Module.Index0Float3Write=true
   )
  ]
  rendererClasses [&NiagaraMeshRendererProperties]
  renderers [{rendererClass=&NiagaraMeshRendererProperties, rendererIndex=0}]
 }
 Sparks {
  emitterUpdateScript [
   &EmitterState(Life Cycle Mode=::NewEnumerator1, Loop Behavior=::NewEnumerator1, Loop Delay=$System.FireDelay, Loop Duration=3, UseLoopDelay=true)
   &SpawnRate(
    SpawnRate=&FloatFromCurve(
     FloatCurve=DataInterface() {propertyValues {Curve {keys [<at 0.239 0>,<at 0.24 1>,<at 0.3813 1>,<at 0.4813 0>]}}}
     CurveIndex=$Emitter.NormalizedLoopAge
     Scale Curve=35
    )
    Use Spawn Probability=false
   )
   &SpawnBurst_Instantaneous(Spawn Count=15, Spawn Time=$System.FireDelay, Use Spawn Probability=false, UseLoopCountLimit=false)
  ]
  particleSpawnScript [
   /SetVariables(Particles.VisibilityTag=&RandomRangeInt(Minimum=0, Maximum=1)) {isSetParameters=true}
   &InitializeParticle(
    Color Mode=::NewEnumerator0
    Lifetime=&MergeHelper_FloatSwitch(
     0=&RandomRangeFloat(Minimum=0.15, Maximum=0.5)
     1=&RandomRangeFloat(Minimum=0.75, Maximum=0.9)
     2=?
     3=?
     Index=$Particles.VisibilityTag
     Num=::NewEnumerator0
    )
    Mass Mode=::NewEnumerator3
    Sprite Rotation Mode=::NewEnumerator2
    Sprite Size=&Multiply_Vector2D_ByFloat(Float=1.6, Vector2D=&RandomRangeVector2D(Minimum=<0.5 1>, Maximum=<2 4>))
    Sprite Size Mode=::NewEnumerator4
    UsePositionOffset=false
    Write Material Random=true
   )
   &ShapeLocation(
    MaskSpawnGroup=false
    Non Uniform Scale=<1.5 0.2 1>
    Offset=<0 -35 60>
    Offset Mode=::NewEnumerator0
    Override Seed=false
    Sphere Radius=80
    UseOwnerScale=true
   )
   &AddVelocity(
    Cone Angle=32
    Cone Axis=<0 0 1>
    Distribution Along Cone Axis=0.25
    Module.UseConeFalloff=true
    Module.UseVelDistribution=true
    Override Seed=false
    Speed Falloff From Cone Axis=0.5
    Velocity Mode=::NewEnumerator2
    Velocity Speed=&RandomRangeFloat(Minimum=100, Maximum=600)
   )
   &CameraOffset(Camera Offset Amount=5)
   &AddVelocity(Velocity=&RandomRangeVector(Minimum=<-20 -20 0>, Maximum=<20 20 20>)) {module=AddVelocity001}
  ]
  particleUpdateScript [
   &ParticleState()
   &ScaleSpriteSize(Uniform Curve Sprite Scale=DataInterface() {propertyValues {Curve {
       defaultValue=1
       keys [
        {interpMode=RCIM_Cubic, tangentMode=RCTM_User, time=0, value=0}
        {interpMode=RCIM_Cubic, tangentMode=RCTM_User, time=0.1931, value=1}
        {arriveTangent=-0.216, interpMode=RCIM_Cubic, leaveTangent=-0.216, tangentMode=RCTM_User, time=0.4834, value=1}
        {tangentMode=RCTM_User, time=1, value=0}
       ]
      }}})
   &Drag(Drag=&RandomRangeFloat(Minimum=1.5, Maximum=5), Use Linear Drag=true, Use Rotational Drag=false)
   &ScaleColor(
    Scale Alpha=&FloatFromCurve(FloatCurve=DataInterface() {propertyValues {Curve {keys [<at -0.0046 1.0113 RCIM_Cubic>,<at 1.0011 0.002 RCIM_Cubic>]}}})
    Scale RGB=&RandomRangeVector(
     Minimum=&Multiply_VectorByFloat(
      Vector=&MakeVectorFromLinearColorRGB(Color=&ColorFromCurve(ColorCurve=DataInterface() {propertyValues {
          AlphaCurve {keys []}
          BlueCurve {keys [<at 0.2378 3.8882>, {arriveTangent=-5.0267, interpMode=RCIM_Cubic, leaveTangent=-5.0267, time=0.6687, value=0.34}, <at 1 0.0568 RCIM_Cubic>]}
          GreenCurve {keys [
            <at 0.2378 12.6481>
            {arriveTangent=-16.2806, interpMode=RCIM_Cubic, leaveTangent=-16.2806, time=0.6687, value=1.3833}
            <at 1 0.239 RCIM_Cubic>
           ]}
          RedCurve {keys [<at 0.2378 25>, {arriveTangent=-31.4877, interpMode=RCIM_Cubic, leaveTangent=-31.4877, time=0.6687, value=5}, <at 1 1 RCIM_Cubic>]}
         }}))
      Float=&RandomRangeFloat(Minimum=1, Maximum=10)
     )
     Maximum=&Multiply_VectorByFloat(
      Vector=&MakeVectorFromLinearColorRGB(Color=&ColorFromCurve(ColorCurve=DataInterface() {propertyValues {
          AlphaCurve {keys []}
          BlueCurve {keys [
            <at 0.0425 4.8877>
            {arriveTangent=-9.8603, interpMode=RCIM_Cubic, leaveTangent=-9.8603, time=0.3499, value=0.3783}
            {arriveTangent=-0.4508, interpMode=RCIM_Cubic, leaveTangent=-0.4508, time=0.5175, value=0.2041}
            <at 0.9011 0.1298>
            <at 0.9527 0.271>
           ]}
          GreenCurve {keys [
            <at 0.0425 10.969>
            {arriveTangent=-22.0738, interpMode=RCIM_Cubic, leaveTangent=-22.0738, time=0.3499, value=1.3506}
            {arriveTangent=-1.7631, interpMode=RCIM_Cubic, leaveTangent=-1.7631, time=0.5175, value=0.4842}
            <at 0.9011 0.3789>
            <at 0.9527 0.3971>
           ]}
          RedCurve {keys [
            <at 0.0425 25>
            {arriveTangent=-48.422, interpMode=RCIM_Cubic, leaveTangent=-48.422, time=0.3499, value=5}
            {arriveTangent=-5.4434, interpMode=RCIM_Cubic, leaveTangent=-5.4434, time=0.5175, value=2}
            <at 0.9011 2>
            <at 0.9527 2>
           ]}
         }}))
      Float=&RandomRangeFloat(Minimum=1, Maximum=10)
     )
    )
    ScaleA=true
    ScaleRGB=true
   )
   &CurlNoiseForce(
    Bias Noise=false
    Noise Frequency=&RandomRangeFloat(Minimum=3, Maximum=25)
    Noise Strength=&RandomRangeFloat(Minimum=300, Maximum=200)
    Pan Noise=false
   )
   &AccelerationForce(Acceleration=&VectorFromCurve(
     VectorCurve=DataInterface() {propertyValues {
       XCurve {defaultValue=1, keys [<at 0.07 2>,<at 0.1 1>,<at 0.25 1>,<at 1 0>]}
       YCurve {keys [<at 0.07 2>,<at 0.1 1>,<at 0.25 1>,<at 1 0>]}
       ZCurve {keys [<at 0.07 2>,<at 0.1 1>,<at 0.25 1>,<at 1 0>]}
      }}
     Scale Curve=&RandomRangeVector(Minimum=<-1 -0.5 3>, Maximum=<1 0.5 3>)
    ))
   &SolveForcesAndVelocity(Clamp Velocity=false, Limit Acceleration=false)
   &ScaleSpriteSizeBySpeed(Max Scale Factor=<1.2 1.5>, Min Scale Factor=<0.7 1>)
   /SetVariables(
    Particles.GlowColor=&MakeLinearColorFromVectorAndFloat(
     "Vector (RGB)"=&MakeVectorFromLinearColorRGB(Color=&ColorFromCurve(ColorCurve=DataInterface() {propertyValues {
         AlphaCurve {keys []}
         BlueCurve {keys [<at 0 0.0893 RCIM_Cubic>,<at 0.3932 0.0885>,<at 0.735 0.0854>,<at 0.9264 0.0717 RCIM_Cubic>]}
         GreenCurve {keys [<at 0 1.019 RCIM_Cubic>,<at 0.3932 0.4775>,<at 0.735 0.2755>,<at 0.9264 0.204 RCIM_Cubic>]}
         RedCurve {keys [<at 0 2 RCIM_Cubic>,<at 0.3932 1>,<at 0.735 1>,<at 0.9264 1 RCIM_Cubic>]}
        }}))
     "Float (Alpha)"=&FloatFromCurve(
      FloatCurve=DataInterface() {propertyValues {Curve {
         defaultValue=1
         keys [{interpMode=RCIM_Cubic, tangentMode=RCTM_User, time=0, value=1}, {interpMode=RCIM_Cubic, tangentMode=RCTM_User, time=1, value=0}]
        }}}
      Scale Curve=1.5
     )
    )
    Particles.GlowSize=&Multiply_Vector2D(A=$Particles.SpriteSize, B=<8 4>)
   ) {isSetParameters=true}
  ]
  rendererClasses [&NiagaraSpriteRendererProperties]
  renderers [
   {rendererClass=&NiagaraSpriteRendererProperties}
   {rendererClass=&NiagaraSpriteRendererProperties, rendererIndex=1}
   {rendererClass=&NiagaraSpriteRendererProperties, rendererIndex=2}
   {rendererClass=&NiagaraSpriteRendererProperties, rendererIndex=3}
  ]
 }
}
summary {
 systemName=NS_Emote_PearlTaffyVista
 emitters [
  {emitterName=MiniGlass, rendererClasses [&NiagaraMeshRendererProperties]}
  {emitterName=FireWall, rendererClasses [&NiagaraSpriteRendererProperties, &NiagaraLightRendererProperties]}
  {emitterName=Sparks, rendererClasses [&NiagaraSpriteRendererProperties]}
  {emitterName=MeshSoftFlame, rendererClasses [&NiagaraMeshRendererProperties]}
  {emitterName=BreakingFlames, rendererClasses [&NiagaraSpriteRendererProperties]}
  {emitterName=Rose, rendererClasses [&NiagaraMeshRendererProperties]}
  {emitterName=Fire, rendererClasses [&NiagaraSpriteRendererProperties]}
  {emitterName=Glows, rendererClasses [&NiagaraSpriteRendererProperties]}
  {emitterName=Glass, rendererClasses [&NiagaraMeshRendererProperties, &NiagaraSpriteRendererProperties]}
 ]
 userVariables [
  User.bIsFrontEnd {type=NiagaraBool}
  User.bIsFrontEndPreview {type=NiagaraBool}
 ]
}
