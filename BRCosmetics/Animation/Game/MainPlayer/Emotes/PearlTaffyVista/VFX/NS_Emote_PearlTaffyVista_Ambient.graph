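# Ambient Niagara effect for the PearlTaffyVista emote: lights, glow, ripples, decals and drifting embers over an 11.567-second loop.
# One emitter uses spawn groups as visibility tags, combining burst particles with rate-spawned embers moved by cone velocity, drag and curl noise.
# Loop-age curves shift colors from violet-blue to orange-red, while a "Lyric Timing" waveform pulses light, glow and decal alpha.
# Ember and ripple brightness adapts to time of day.
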NiagaraSystem /BRCosmetics/Animation/Game/MainPlayer/Emotes/PearlTaffyVista/VFX/NS_Emote_PearlTaffyVista_Ambient
dependencies {
 usedDataInterfaces [&NiagaraDataInterfaceCurve, &NiagaraDataInterfaceColorCurve]
 usedDynamicInputs []
 usedModules [&SystemState, &BlendSpaceHelper]
}
emitters {Minimal {
  emitterUpdateScript [
   &EmitterState(Life Cycle Mode=::NewEnumerator1, Loop Duration=11.567, UseLoopDelay=false)
   &SpawnBurst_Instantaneous(Use Spawn Probability=false, UseLoopCountLimit=false) {module=SpawnBurst_Instantaneous001}
   &SpawnBurst_Instantaneous(Spawn Group=1, Use Spawn Probability=false, UseLoopCountLimit=false)
   &SpawnBurst_Instantaneous(Spawn Group=3, Use Spawn Probability=false, UseLoopCountLimit=false) {module=SpawnBurst_Instantaneous002}
   &SpawnRate(SpawnGroup=4, SpawnRate=30, Use Spawn Probability=false)
  ]
  particleSpawnScript [
   /SetVariables(Particles.VisibilityTag=&SpawnGroup()) {isSetParameters=true}
   &InitializeParticle(
    Color Mode=::NewEnumerator0
    Lifetime=&MaskFloat(
     Mask Source=0
     Mask Parameter=?
     Mask=4
     Masked Float=&RandomRangeFloat(Minimum=0.1, Maximum=1)
     Passthrough Float=11.567
    )
    Position Offset=&MaskVector(Mask Source=0, Mask Parameter=?, Mask=0, Masked Vector=<0 0 0>, Passthrough Vector=<0 0 1>)
    Sprite Rotation Angle=&MaskFloat(
     Mask Source=0
     Mask Parameter=?
     Mask=4
     Masked Float=&RandomRangeFloat(Minimum=-360, Maximum=360)
     Passthrough Float=0
    )
    Sprite Rotation Mode=::NewEnumerator0
    Sprite Size Mode=::NewEnumerator3
    Uniform Sprite Size=&MaskFloat(
     Mask Source=0
     Mask Parameter=?
     Mask=4
     Masked Float=&RandomRangeFloat(Minimum=6, Maximum=12)
     Passthrough Float=110
    )
    UsePositionOffset=true
    Write Material Random=true
   )
   &ShapeLocation(
    Handle Radius:NiagaraFloat=&MaskFloat(Mask Source=0, Mask Parameter=?, Mask=4, Masked Float=10, Passthrough Float=0)
    Large Radius=&MaskFloat(Mask Source=0, Mask Parameter=?, Mask=4, Masked Float=55, Passthrough Float=0)
    MaskSpawnGroup=false
    Override Seed=false
    Shape Primitive=::NewEnumerator3
    UseOwnerScale=true
   )
   &AddVelocity(
    Cone Angle=32
    Cone Axis=<0 0 1>
    Distribution Along Cone Axis=0.25
    Module.UseConeFalloff=true
    Module.UseVelDistribution=true
    Override Seed=false
    Speed Falloff From Cone Axis=0.5
    Velocity Mode=::NewEnumerator2
    Velocity Speed=&MaskFloat(
     Mask Source=0
     Mask Parameter=?
     Mask=4
     Masked Float=&RandomRangeFloat(Minimum=100, Maximum=200)
     Passthrough Float=0
    )
   )
   &ApplyInitialForces()
   &CameraOffset(Camera Offset Amount=&MergeHelper_FloatSwitch(
     0=0
     1=&MakeCustomFloatFromBool(
      Bool=&SetBoolByIntComparison(A=$Engine.QualityLevel, B=1, Comparison Type=::NewEnumerator7)
      True Float=-50
      False Float=-5
     )
     2=0
     3=0
     4=0
     Index=&SpawnGroup()
     Num=::NewEnumerator3
    ))
   /SetVariables(Particles.RipplesFacing:Vector3f=<0 0 1>) {isSetParameters=true}
  ]
  particleUpdateScript [
   &ParticleState()
   &GravityForce(Gravity=&MaskVector(
     Mask Source=0
     Mask Parameter=?
     Mask=4
     Masked Vector=&RandomRangeVector(Minimum=<-20 -20 -20>, Maximum=<20 20 20>)
     Passthrough Vector=<0 0 0>
    ))
   /SetVariables(Particles.Lyric Timing:NiagaraFloat=&Waveform("Waveform [1]"=::NewEnumerator0, "[1] Frequency"=1.3575, "[1] Amplitude Min/Max"=<0.15 1>)) {isSetParameters=true}
   &Light_Attributes(
    Diffuse Scale=1
    Falloff Exponent=1
    Light Position Offset:Vector3f=<0 -30 100>
    Position=1
    Position to Offset:NiagaraPosition=$Engine.Emitter.SimulationPosition
    Radius=1
    Specular Scale=1
    UsePositionOffset=true
    Volumetric Scattering=1
   )
   /SetVariables(
    Particles.Ashes Dyn Param:Vector4f=&MakeVector4(X=1, Y=0, Z=0, W=0)
    Particles.Ashes SpriteSize:Vector2f=&Vector2DFromFloat(Value=&RandomRangeFloat(Minimum=1, Maximum=5))
    "Particles.Ashes color ":LinearColor=&MakeLinearColorFromVectorAndFloat(
     "Vector (RGB)"=&MakeVectorFromLinearColorRGB(Color=&LerpLinearColors(
       StartColor=&FNTimeOfDayColor(
        Night Color=&RandomRangeLinearColor(Minimum=<rgba 7.9922 0.6509 25 1>, Maximum=<rgba 7.3395 0.256 25 1>)
        Morning Color=&RandomRangeLinearColor(Minimum=<rgba 17.9018 2.9687 50 1>, Maximum=<rgba 14.679 0.512 50 1>)
        Day Color=&RandomRangeLinearColor(Minimum=<rgba 19.3236 5.0521 50 1>, Maximum=<rgba 14.679 0.512 50 1>)
        Evening Color=&RandomRangeLinearColor(Minimum=<rgba 17.9018 2.9687 50 1>, Maximum=<rgba 14.679 0.512 50 1>)
       )
       EndColor=&FNTimeOfDayColor(
        Night Color=&RandomRangeLinearColor(Minimum=<rgba 25 2.7697 1.1928 1>, Maximum=<rgba 25 3.0199 0.1 1>)
        Morning Color=&RandomRangeLinearColor(Minimum=<rgba 50 5.8929 2.3561 1>, Maximum=<rgba 50 6.0397 0.2 1>)
        Day Color=&RandomRangeLinearColor(Minimum=<rgba 50 6.4576 4.4516 1>, Maximum=<rgba 50 6.0397 0.2 1>)
        Evening Color=&RandomRangeLinearColor(Minimum=<rgba 50 5.9384 2.7845 1>, Maximum=<rgba 50 6.0397 0.2 1>)
       )
       LerpFactor=&FloatFromCurve(
        FloatCurve=DataInterface() {propertyValues {Curve {
           defaultValue=1
           keys [{interpMode=RCIM_Cubic, tangentMode=RCTM_User, time=0.2, value=0}, {interpMode=RCIM_Cubic, tangentMode=RCTM_User, time=0.75, value=1}]
          }}}
        CurveIndex=$Emitter.NormalizedLoopAge
        Scale Curve=&FNTimeOfDayScalar(Night Scale=0.75, Morning Scale=1, Evening Scale=1)
       )
      ))
     "Float (Alpha)"=&FloatFromCurve(FloatCurve=DataInterface() {propertyValues {Curve {keys [<at 0 1>,<at 0.5476 0.2284>,<at 1 0>]}}})
    )
   ) {isSetParameters=true}
   &Drag(
    Drag=&MaskFloat(
     Mask Source=0
     Mask Parameter=?
     Mask=4
     Masked Float=&RandomRangeFloat(Minimum=1.5, Maximum=2)
     Passthrough Float=0
    )
    Use Linear Drag=true
    Use Rotational Drag=false
   )
   &CurlNoiseForce(
    Bias Noise=false
    Noise Frequency=&MaskFloat(
     Mask Source=0
     Mask Parameter=?
     Mask=4
     Masked Float=&RandomRangeFloat(Minimum=5, Maximum=20)
     Passthrough Float=0
    )
    Noise Strength=&MaskFloat(
     Mask Source=0
     Mask Parameter=?
     Mask=4
     Masked Float=&RandomRangeFloat(Minimum=50, Maximum=120)
     Passthrough Float=0
    )
    Pan Noise=false
   )
   /SetVariables(Particles.Flame Light Color:LinearColor=&MakeLinearColorFromVectorAndFloat(
     "Vector (RGB)"=&MakeVectorFromLinearColorRGB(Color=&ColorFromCurve(
       ColorCurve=DataInterface() {propertyValues {
         AlphaCurve {keys []}
         BlueCurve {keys [<at 0 5 RCIM_Cubic>,<at 0.762 0 RCIM_Cubic>]}
         GreenCurve {keys [<at 0 0.2653 RCIM_Cubic>,<at 0.762 0.2131 RCIM_Cubic>]}
         RedCurve {keys [<at 0 1.6803 RCIM_Cubic>,<at 0.762 5 RCIM_Cubic>]}
        }}
       CurveIndex=$Emitter.NormalizedLoopAge
      ))
     "Float (Alpha)"=$Particles.Lyric Timing
    )) {enabled=false, isSetParameters=true}
   &Light_Attributes(
    Diffuse Scale=1
    Falloff Exponent=1
    Light Position Offset:Vector3f=<0 60 115>
    Position=1
    Position to Offset:NiagaraPosition=$Engine.Emitter.SimulationPosition
    Radius=1
    Specular Scale=1
    UsePositionOffset=true
    Volumetric Scattering=1
   ) {module=Light_Attributes001}
   /SetVariables(Particles.Front Light Color:LinearColor=&MakeLinearColorFromVectorAndFloat(
     "Vector (RGB)"=&MakeVectorFromLinearColorRGB(Color=&ColorFromCurve(
       ColorCurve=DataInterface() {propertyValues {
         AlphaCurve {keys []}
         BlueCurve {keys [
           <at 0 5 RCIM_Cubic>
           {arriveTangent=-5.8912, interpMode=RCIM_Cubic, leaveTangent=-5.8912, time=0.762, value=0}
           <at 0.8599 -0.0661>
           <at 0.9505 0>
           <at 0.994 0>
          ]}
         GreenCurve {keys [
           <at 0 0.869 RCIM_Cubic>
           {arriveTangent=-0.7204, interpMode=RCIM_Cubic, leaveTangent=-0.7204, time=0.762, value=0.2131}
           <at 0.8599 0.2495>
           <at 0.9505 1.843>
           <at 0.994 0.3686>
          ]}
         RedCurve {keys [
           <at 0 1.901 RCIM_Cubic>
           {arriveTangent=3.6514, interpMode=RCIM_Cubic, leaveTangent=3.6514, time=0.762, value=5}
           <at 0.8599 5.041>
           <at 0.9505 5>
           <at 0.994 1>
          ]}
        }}
       CurveIndex=$Emitter.NormalizedLoopAge
      ))
     "Float (Alpha)"=$Particles.Lyric Timing
    )) {isSetParameters=true}
   /SetVariables(
    Particles.Glow Color=&MakeLinearColorFromVectorAndFloat(
     "Vector (RGB)"=&MakeVectorFromLinearColorRGB(Color=&ColorFromCurve(
       ColorCurve=DataInterface() {propertyValues {
         AlphaCurve {keys []}
         BlueCurve {keys [<at 0 2 RCIM_Cubic>, {arriveTangent=-2.4357, interpMode=RCIM_Cubic, leaveTangent=-2.4357, time=0.6518, value=0}, <at 0.8211 0>, <at 1 0>]}
         GreenCurve {keys [
           <at 0 0.328 RCIM_Cubic>
           {arriveTangent=-0.0229, interpMode=RCIM_Cubic, leaveTangent=-0.0229, time=0.6518, value=0.2315}
           <at 0.8211 0.3092>
           <at 1 0.2588>
          ]}
         RedCurve {keys [
           <at 0 0.7348 RCIM_Cubic>
           {arriveTangent=1.5407, interpMode=RCIM_Cubic, leaveTangent=1.5407, time=0.6518, value=2}
           <at 0.8211 2>
           <at 1 1.4479>
          ]}
        }}
       CurveIndex=$Emitter.NormalizedLoopAge
      ))
     "Float (Alpha)"=&Multiply_Float(A=$Particles.Lyric Timing, B=0.05)
    )
    Particles.Glow Pos:NiagaraPosition=&AddVectorToPosition(Position=$Engine.Emitter.SimulationPosition, Vector=<0 0 85>)
    Particles.Glow Size=<350 400>
   ) {isSetParameters=true}
   /SetVariables(
    Particles.EmberGlowColor:LinearColor=&MakeLinearColorFromVectorAndFloat(
     "Vector (RGB)"=&MakeVectorFromLinearColorRGB(Color=&LerpLinearColors(
       StartColor=<rgba 0.7295 0.3186 1.2 1>
       EndColor=<rgba 1.2 0.3146 0.0402 1>
       LerpFactor=&FloatFromCurve(
        FloatCurve=DataInterface() {propertyValues {Curve {
           defaultValue=1
           keys [{interpMode=RCIM_Cubic, tangentMode=RCTM_User, time=0.2, value=0}, {interpMode=RCIM_Cubic, tangentMode=RCTM_User, time=0.75, value=1}]
          }}}
        CurveIndex=$Emitter.NormalizedLoopAge
       )
      ))
     "Float (Alpha)"=&FloatFromCurve(
      FloatCurve=DataInterface() {propertyValues {Curve {
         defaultValue=1
         keys [{interpMode=RCIM_Cubic, tangentMode=RCTM_User, time=0, value=1}, {interpMode=RCIM_Cubic, tangentMode=RCTM_User, time=1, value=0}]
        }}}
      Scale Curve=&FNTimeOfDayScalar(Night Scale=0.8, Morning Scale=1, Evening Scale=1)
     )
    )
    Particles.EmberGlowSize:Vector2f=&Multiply_Vector2D_ByFloat(Float=6, Vector2D=$Particles.Ashes SpriteSize)
   ) {isSetParameters=true}
   /Niagara/Modules/Decal/Decal_Attributes(
    Decal Facing Mode=1
    Decal Rotation Mode=2
    Decal Size:Vector3f=<20 150 150>
    Fade=1
    Orientation=1
    Rotation Angle:NiagaraFloat=&FloatFromCurve(
     FloatCurve=DataInterface() {propertyValues {Curve {defaultValue=1, keys [{tangentMode=RCTM_User, time=0, value=1}, {tangentMode=RCTM_User, time=1, value=0}]}}}
     CurveIndex=$Emitter.NormalizedLoopAge
     Scale Curve=2500
    )
    Set Alignment=false
    Size=1
   )
   /SetVariables(
    Particles.Decal Color:LinearColor=&MakeLinearColorFromVectorAndFloat(
     "Vector (RGB)"=&MakeVectorFromLinearColorRGB(Color=&ColorFromCurve(
       ColorCurve=DataInterface() {propertyValues {
         AlphaCurve {keys []}
         BlueCurve {keys [<at 0 5 RCIM_Cubic>,<at 0.7521 0 RCIM_Cubic>]}
         GreenCurve {keys [<at 0 0.7071 RCIM_Cubic>,<at 0.7521 0.2874 RCIM_Cubic>]}
         RedCurve {keys [<at 0 1.5119 RCIM_Cubic>,<at 0.7521 5 RCIM_Cubic>]}
        }}
       CurveIndex=$Emitter.NormalizedLoopAge
      ))
     "Float (Alpha)"=&FloatFromCurve(
      FloatCurve=DataInterface() {propertyValues {Curve {keys [
          <at -0.0109 0 RCIM_Cubic>
          {arriveTangentWeight=0.0617, interpMode=RCIM_Cubic, leaveTangentWeight=0.3716, tangentMode=RCTM_User, time=0.0759, value=1}
          {arriveTangentWeight=0.3789, interpMode=RCIM_Cubic, leaveTangentWeight=0.0804, tangentMode=RCTM_User, time=0.9288, value=1}
          <at 1 0 RCIM_Cubic>
         ]}}}
      Scale Curve=&Multiply_Float(A=0.12, B=$Particles.Lyric Timing)
     )
    )
    Particles.Decal Color 1:LinearColor=&MakeLinearColorFromVectorAndFloat(
     "Vector (RGB)"=&MakeVectorFromLinearColorRGB(Color=&ColorFromCurve(
       ColorCurve=DataInterface() {propertyValues {
         AlphaCurve {keys []}
         BlueCurve {keys [
           <at 0 1 RCIM_Cubic>
           {arriveTangent=-1.2827, interpMode=RCIM_Cubic, leaveTangent=-1.2827, time=0.6706, value=0.2065}
           <at 0.7796 0>
           <at 0.964 0>
          ]}
         GreenCurve {keys [
           <at 0 0.4247 RCIM_Cubic>
           {arriveTangent=-0.1107, interpMode=RCIM_Cubic, leaveTangent=-0.1107, time=0.6706, value=0.2773}
           <at 0.7796 0.3384>
           <at 0.964 0.6507>
          ]}
         RedCurve {keys [
           <at 0 0.503 RCIM_Cubic>
           {arriveTangent=0.6375, interpMode=RCIM_Cubic, leaveTangent=0.6375, time=0.6706, value=1}
           <at 0.7796 1>
           <at 0.964 2>
          ]}
        }}
       CurveIndex=$Emitter.NormalizedLoopAge
      ))
     "Float (Alpha)"=&Multiply_Float(
      A=$Particles.Lyric Timing
      B=&FloatFromCurve(
       FloatCurve=DataInterface() {propertyValues {Curve {keys [
           <at -0.0109 0 RCIM_Cubic>
           {arriveTangentWeight=0.0617, interpMode=RCIM_Cubic, leaveTangentWeight=0.3716, tangentMode=RCTM_User, time=0.0315, value=1}
           {arriveTangentWeight=0.3789, interpMode=RCIM_Cubic, leaveTangentWeight=0.0804, tangentMode=RCTM_User, time=0.8957, value=1}
           <at 1 0 RCIM_Cubic>
          ]}}}
       Scale Curve=0.1
      )
     )
    )
    Particles.Decal Color 2:LinearColor=&MakeLinearColorFromVectorAndFloat(
     "Vector (RGB)"=&MakeVectorFromLinearColorRGB(Color=&ColorFromCurve(
       ColorCurve=DataInterface() {propertyValues {
         AlphaCurve {keys []}
         BlueCurve {keys [<at 0 5 RCIM_Cubic>,<at 0.7425 0 RCIM_Cubic>]}
         GreenCurve {keys [<at 0 0.7071 RCIM_Cubic>,<at 0.7425 0.2874 RCIM_Cubic>]}
         RedCurve {keys [<at 0 1.5119 RCIM_Cubic>,<at 0.7425 5 RCIM_Cubic>]}
        }}
       CurveIndex=$Emitter.NormalizedLoopAge
      ))
     "Float (Alpha)"=&FloatFromCurve(
      FloatCurve=DataInterface() {propertyValues {Curve {keys [
          <at -0.0109 0 RCIM_Cubic>
          {arriveTangentWeight=0.0617, interpMode=RCIM_Cubic, leaveTangentWeight=0.3716, tangentMode=RCTM_User, time=0.0759, value=1}
          {arriveTangentWeight=0.3789, interpMode=RCIM_Cubic, leaveTangentWeight=0.0804, tangentMode=RCTM_User, time=0.9288, value=1}
          <at 1 0 RCIM_Cubic>
         ]}}}
      Scale Curve=&Multiply_Float(A=0.03, B=$Particles.Lyric Timing)
     )
    )
   ) {isSetParameters=true}
   &Decal_Attributes(Decal Size:Vector3f=<20 600 600>, Fade=1, Size=1) {module=Decal_Attributes001}
   &BlendSpace(Position Mode=2, Rotation Contribution=0)
   &Decal_Attributes(
    Decal Facing Mode=1
    Decal Rotation Mode=2
    Decal Size:Vector3f=<20 165 165>
    Fade=1
    Orientation=1
    Rotation Angle:NiagaraFloat=&FloatFromCurve(
     FloatCurve=DataInterface() {propertyValues {Curve {defaultValue=1, keys [{tangentMode=RCTM_User, time=0, value=1}, {tangentMode=RCTM_User, time=1, value=0}]}}}
     CurveIndex=$Emitter.NormalizedLoopAge
     Scale Curve=2500
    )
    Set Alignment=false
    Size=1
   ) {module=Decal_Attributes002}
   /SetVariables(Particles.RipplesColor:LinearColor=&MakeLinearColorFromVectorAndFloat(
     "Vector (RGB)"=&MakeVectorFromLinearColorRGB(Color=&ColorFromCurve(
       ColorCurve=DataInterface() {propertyValues {
         AlphaCurve {keys []}
         BlueCurve {keys [
           <at 0 1 RCIM_Cubic>
           <at 0.305 0.5679>
           {arriveTangent=-0.9776, interpMode=RCIM_Cubic, leaveTangent=-0.9776, time=0.7049, value=0.1301}
           <at 0.869 0.1505>
           <at 0.9656 0.1505>
          ]}
         GreenCurve {keys [
           <at 0 0.2094 RCIM_Cubic>
           <at 0.305 0.1464>
           {arriveTangent=-0.1346, interpMode=RCIM_Cubic, leaveTangent=-0.1346, time=0.7049, value=0.0986}
           <at 0.869 0.0924>
           <at 0.9656 0.0924>
          ]}
         RedCurve {keys [
           <at 0 0.8476 RCIM_Cubic>
           <at 0.305 1>
           {arriveTangent=0.7507, interpMode=RCIM_Cubic, leaveTangent=0.7507, time=0.7049, value=1.5}
           <at 0.869 1.5>
           <at 0.9656 1.5>
          ]}
        }}
       CurveIndex=$Emitter.NormalizedLoopAge
      ))
     "Float (Alpha)"=&FloatFromCurve(
      FloatCurve=DataInterface() {propertyValues {Curve {keys [
          <at 0 0 RCIM_Cubic>
          {arriveTangentWeight=0.0502, interpMode=RCIM_Cubic, leaveTangentWeight=3.7644, tangentMode=RCTM_User, time=0.3, value=1}
          {arriveTangentWeight=3.7644, interpMode=RCIM_Cubic, leaveTangentWeight=0.0559, tangentMode=RCTM_User, time=11.25, value=1}
          <at 11.567 0 RCIM_Cubic>
         ]}}}
      CurveIndex=$Particles.Age
      Scale Curve=&FNTimeOfDayScalar(Night Scale=0.37, Morning Scale=0.55, Day Scale=0.55, Evening Scale=0.55)
     )
    )) {isSetParameters=true}
   &SolveForcesAndVelocity(Clamp Velocity=false, Limit Acceleration=false)
  ]
  rendererClasses [
   &NiagaraLightRendererProperties
   &NiagaraSpriteRendererProperties
   &NiagaraDecalRendererProperties
  ]
  renderers [
   {rendererClass=&NiagaraLightRendererProperties, rendererIndex=0}
   {rendererClass=&NiagaraLightRendererProperties, rendererIndex=1}
   {rendererClass=&NiagaraSpriteRendererProperties, rendererIndex=2}
   {rendererClass=&NiagaraDecalRendererProperties, rendererIndex=3}
   {rendererClass=&NiagaraDecalRendererProperties, rendererIndex=4}
   {rendererClass=&NiagaraDecalRendererProperties, rendererIndex=5}
   {rendererClass=&NiagaraSpriteRendererProperties, rendererIndex=6}
   {rendererClass=&NiagaraSpriteRendererProperties, rendererIndex=7}
   {rendererClass=&NiagaraSpriteRendererProperties, rendererIndex=8}
  ]
 }}
summary {
 systemName=NS_Emote_PearlTaffyVista_Ambient
 emitters [{
   emitterName=Minimal
   rendererClasses [
    &NiagaraLightRendererProperties
    &NiagaraSpriteRendererProperties
    &NiagaraDecalRendererProperties
   ]
  }]
 userVariables []
}
